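# Ambient Niagara effect for the Sprite Haven cave, filling it with teal-green drifting motes and fluttering butterflies.
# OnTree spawns GPU sprites in user colors, moved by noise, wind and upward gravity, with region masks and distance-based shrinking.
# Butterflies001 pairs meshes and sprites, using waveform scaling for flutter, curl noise and distance-field avoidance.
# A camera-attracted butterfly variant and a fog emitter are present but disabled.
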NiagaraSystem /Game/Environments/Hera/FX/SpriteHaven/NS_SpriteHavenCave_Ambient
dependencies {
 usedDataInterfaces [
  &NiagaraDataInterfaceCamera
  &NiagaraDataInterfaceCurve
  &NiagaraDataInterfaceColorCurve
  &NiagaraDataInterfaceCollisionQuery
  &NiagaraDataInterfaceGBuffer
  &NiagaraDataInterfaceLandscape
  &NiagaraDataInterfaceSpriteRendererInfo
 ]
 usedDynamicInputs [&Vector2DFromFloat, /Game/Effects/Niagara/DynamicInputs/Math/FN_Fit]
 usedModules [
  &self:SystemSpawnScript.SystemScriptSource.SystemScriptGraph.NiagaraNodeAssignment_0.SetVariables_#1
  &SystemState
  &self:SystemSpawnScript.SystemScriptSource.SystemScriptGraph.NiagaraNodeAssignment_1.SetVariables_#2
 ]
}
emitters {
 Butterflies {
  bEnabled=false
  simTarget=GPUComputeSim
  emitterUpdateScript [
   /SetVariables() {isSetParameters=true, module=SetVariables_#3}
   /SetVariables(
    Emitter.CameraForward:Vector3f=&ReturnCameraVector(Camera Property=::NewEnumerator1, Coordinate Space=::Simulation, CameraQuery=$Emitter.CameraRead)
    Emitter.CameraPosition:Vector3f=&ReturnCameraVector(Camera Property=::NewEnumerator0, Coordinate Space=::World, CameraQuery=$Emitter.CameraRead)
    Emitter.CameraRight:Vector3f=&ReturnCameraVector(Camera Property=::NewEnumerator3, Coordinate Space=::Simulation, CameraQuery=$Emitter.CameraRead)
    Emitter.CameraUp:Vector3f=&ReturnCameraVector(Camera Property=::NewEnumerator2, Coordinate Space=::Simulation, CameraQuery=$Emitter.CameraRead)
   ) {isSetParameters=true, module=SetVariables_#4}
   &EmitterState(
    Enable Distance Culling=true
    Life Cycle Mode=::NewEnumerator1
    Loop Behavior=::NewEnumerator1
    Loop Duration Mode=::NewEnumerator1
    Max Distance Response=::SleepAndClearParticles
    Module.bApplySpawnCountScaleByDistance=true
    Module.bUseMaxDistance=true
    Module.bUseMinDistance=false
    Reset Age On Awaken=true
    Scalability Mode=::NewEnumerator1
    Scale Spawn Count=true
    Spawn Count Scale By Distance=DataInterface() {propertyValues {Curve {keys [<at 0.5 1>,<at 0.55 1>,<at 1 1>]}}}
    UseLoopDelay=false
   )
   &SpawnPerFrame(
    Spawn:NiagaraBool=&SetBoolByFloatComparison(A=$Emitter.Age, B=1, Comparison Type=::NewEnumerator6)
    Spawn Count:NiagaraInt32=/Niagara/DynamicInputs/Math/Min_Int(
     A:NiagaraInt32=500
     B:NiagaraInt32=/Niagara/DynamicInputs/Subtract/Subtract_Integer(A:NiagaraInt32=50000, B:NiagaraInt32=$Engine.Emitter.NumParticles)
    )
    Use Spawn Probability=false
   ) {enabled=false}
   &SpawnRate(
    SpawnRate=&Multiply_Float(
     A=1000
     B=&FloatFromCurve(FloatCurve=DataInterface() {propertyValues {Curve {keys [<at 0 1>,<at 1 0>]}}}, CurveIndex=$Emitter.Age)
    )
    Use Spawn Probability=false
   )
  ]
  particleSpawnScript [
   &InitializeParticle(
    Color=&ColorFromCurve(
     ColorCurve=DataInterface() {propertyValues {
       AlphaCurve {keys [<at 0 1 RCIM_Constant>,<at 1 1 RCIM_Constant>]}
       BlueCurve {keys [<at 0.0989 0.9911 RCIM_Constant>,<at 0.3996 0.8153 RCIM_Constant>,<at 0.6764 0.3907>]}
       GreenCurve {keys [<at 0.0989 0.9447 RCIM_Constant>,<at 0.3996 0.8906 RCIM_Constant>,<at 0.6764 0.8021>]}
       RedCurve {keys [<at 0.0989 0.0929 RCIM_Constant>,<at 0.3996 0.0009 RCIM_Constant>,<at 0.6764 0.0008>]}
      }}
     CurveIndex=&RandomRangeFloat(Minimum=0, Maximum=1)
    )
    Lifetime=&RandomRangeFloat(Minimum=3, Maximum=5)
    Mesh Scale Mode=::NewEnumerator1
    Mesh Uniform Scale Max=4
    Mesh Uniform Scale Min=2
    Sprite Rotation Mode=::NewEnumerator2
    Sprite Size Mode=::NewEnumerator3
    Uniform Sprite Size=80
    UsePositionOffset=false
    Write Material Random=true
   ) {module=InitializeParticle001}
   &ShapeLocation(MaskSpawnGroup=false, Offset Mode=::NewEnumerator0, Override Seed=false, Sphere Radius=2000, UseOwnerScale=true)
   /SetVariables(Particles.Position=&Add_Vector(
     A=Linked() {linkedVariable {name=Particles.Position, type=NiagaraPosition}}
     B=&MakeVector(X=0, Y=0, Z=&Waveform("Waveform [1]"=::NewEnumerator0, "[1] Frequency"=5, "[1] Amplitude Min/Max"=<0 25>))
    )) {isSetParameters=true}
   /SetVariables(Particles.New Vector 3f:Vector3f=<0 0 1>) {isSetParameters=true}
   /SetVariables() {isSetParameters=true}
   /SetVariables(Particles.InitialPos:NiagaraPosition=$Particles.Position) {isSetParameters=true}
   /Niagara/Modules/Math/PlaceParticlesOnDepthBuffer_GPU(Max Z Depth=10000, Module.Use Max Z Depth=true) {enabled=false}
   /Niagara/Modules/Update/Utility/SampleGBuffer(Kill Beyond Depth=true, Max Z Depth=10000, Snap To Depth:ENiagaraGBufferSnapToDepth=::NewEnumerator1, Write To Color=false)
   /Niagara/Modules/Math/MoveToNearestDistanceFieldSurface_GPU(Kill by Distance=8000, Radius=500, bKillByDistance=false) {enabled=false}
   &InitialMeshOrientation(
    Mesh Axis To Orient=&RandomRangeVector(Minimum=<0 0 0>, Maximum=<1 1 1>)
    Mesh Orientation Mode=::NewEnumerator2
    Orientation Vector:NiagaraPosition=$Output.MoveToNearestDistanceFieldSurface_GPU.SurfacePositionWithoutRadiusOffset
    Rotation=&RandomRangeVector(Minimum=<0 0 0>, Maximum=<0 0 1>)
    Rotation Coordinate Space=::NewEnumerator2
    Use Rotation Vector=true
   )
   &CameraOffset(Camera Offset Amount=100)
   &KillParticles(Kill Particles=&SetBoolByFloatComparison(
     A=&VectorLength(Vector=&Subtract_Vector(A=$Particles.InitialPos, B=Linked() {linkedVariable {name=Particles.Position, type=NiagaraPosition}}))
     B=8000
    )) {enabled=false, module=KillParticles001}
   &KillParticles(Kill Particles=&Boolean_NotOperation(Bool=$Output.PlaceParticlesOnDepthBuffer_GPU.ValidSample)) {enabled=false, module=KillParticles002}
   /Game/Effects/Niagara/Modules/Spawn/FN_Spawn_GBuffer_Color_WithinRangeV2(HueMax=110, HueMin=90) {enabled=false}
   &CameraOffset(Camera Offset Amount=100) {module=CameraOffset001}
   &AddVelocity(Velocity=<0 0 0>)
   &KillParticlesInVolume(Invert Volume=true, Sphere Radius=2500) {module=KillParticlesInVolume001}
  ]
  particleUpdateScript [
   &ParticleState(DeltaTime:NiagaraFloat=&MakeCustomFloatFromBool(
     Bool=&SetBoolByFloatComparison(A=&VectorLength(Vector=$Particles.Velocity), B=1)
     True Float=$Engine.DeltaTime
     False Float=0
    ))
   &ScaleSpriteSize()
   /Niagara/Modules/Update/Forces/PointAttractionForce(
    Attraction Radius=250
    AttractionStrength:NiagaraFloat=&Multiply_Float(
     A=&FloatFromCurve(
      FloatCurve=DataInterface() {propertyValues {Curve {keys [<at 0 1>,<at 0.25 1>]}}}
      Scale Curve=&RandomRangeFloat(Minimum=-3, Maximum=-5)
     )
     B=1
    )
    AttractorPosition:Vector3f=&Add_Vector(
     A=$Emitter.CameraPosition
     B=&Add_Vector(
      A=&Add_Vector(
       A=&Multiply_VectorByFloat(Vector=<0 0 1>, Float=-100)
       B=&Multiply_VectorByFloat(Vector=$Emitter.CameraForward, Float=-250)
      )
      B=&Multiply_VectorByFloat(Vector=$Emitter.CameraRight, Float=0)
     )
    )
    Falloff Exponent=1
    Module.Kill Within Radius=false
    Module.Use Falloff=false
   )
   /SetVariables(Particles.Normalized Velocity:Vector3f=$Particles.Velocity) {isSetParameters=true}
   /SetVariables(Particles.Normalized Velocity:Vector3f=&self:ScratchDynamicInput_03(Default:Vector3f=$Particles.Velocity)) {enabled=false, isSetParameters=true}
   /SetVariables(Particles.Is Active:NiagaraBool=&Boolean_OrOperation(
     Bool1=$Particles.Is Active
     Bool2=&SetBoolByFloatComparison(A=&VectorLength(Vector=$Particles.Normalized Velocity), B=1)
    )) {isSetParameters=true}
   &CurlNoiseForce(
    Curl Noise Cone Mask Angle=90
    Mask Curl Noise=true
    Noise Frequency=&RandomRangeFloat(Minimum=0.85, Maximum=1)
    Noise Strength=&MakeCustomFloatFromBool(Bool=$Particles.Is Active, True Float=&RandomRangeFloat(Minimum=500, Maximum=2500), False Float=0)
    Pan Noise=false
    Pan Noise Field=<0 0 0.1>
   )
   &GravityForce(Gravity=&Multiply_VectorByFloat(
     Vector=<0 0 1>
     Float=&MakeCustomFloatFromBool(Bool=$Particles.Is Active, True Float=&RandomRangeFloat(Minimum=500, Maximum=1500), False Float=0)
    ))
   &CurlNoiseForce(
    Noise Frequency=&RandomRangeFloat(Minimum=2, Maximum=3)
    Noise Strength=&MakeCustomFloatFromBool(Bool=$Particles.Is Active, True Float=&RandomRangeFloat(Minimum=2500, Maximum=5000), False Float=0)
    Pan Noise=true
    Pan Noise Field=<0 0 1>
   ) {module=CurlNoiseForce001}
   &PointAttractionForce(
    Attraction Radius=1000
    AttractionStrength:NiagaraFloat=&Multiply_Float(
     A=&MakeCustomFloatFromBool(
      Bool=$Particles.Is Active
      True Float=0
      False Float=&FloatFromCurve(FloatCurve=DataInterface() {propertyValues {Curve {keys [<at 0 1>,<at 0.25 0>]}}}, Scale Curve=-8)
     )
     B=1
    )
    AttractorPosition:Vector3f=&self:ScratchDynamicInput(
     CameraPosition:NiagaraPosition=Linked() {linkedVariable {name=Emitter.CameraPosition, type=Vector3f}}
     CameraForward:Vector3f=$Emitter.CameraForward
     HeightOffset:NiagaraFloat=&RandomRangeFloat(Minimum=-100, Maximum=-150)
     CameraForwardOffset:NiagaraFloat=&RandomRangeFloat(Minimum=200, Maximum=250)
    )
    Falloff Exponent=1
    Module.Kill Within Radius=false
    Module.Use Falloff=false
   ) {module=PointAttractionForce001}
   &ScaleMeshSize(Scale Factor=&MakeVector(
     X=1
     Y=1
     Z=&Waveform(
      "Waveform [1]"=::NewEnumerator0
      "[1] Phase Scale"=&RandomRangeFloat(Minimum=0.25, Maximum=3)
      "[1] Frequency"=1
      "[1] Amplitude Min/Max"=<0.7 2>
     )
    )) {module=ScaleMeshSize001}
   &KillParticles(Kill Particles=&self:ScratchDynamicInput_01())
   &Collision(
    Age Colliding Particles=false
    Collision Enabled=&SetBoolByFloatComparison(A=$Particles.Age, B=0.25)
    Enable Rest State=false
    Friction=0
    GPU Collision Type=::NewEnumerator2
    Kill Particles Lodged Within Meshes=false
    KillOnCollision=false
    Max Number Of Collisions=2
    Module.EnableMaxCollisionCount=false
    Module.Randomize Collision Normal Vector=false
    Particle Radius Scale=10
    Randomize Collision Normal=0
    Restitution=1
    Simple Friction=false
    Static Friction=0
    Static Friction Engagement Speed=0
   ) {enabled=false}
   /Niagara/Modules/Math/AvoidDistanceFieldSurfaces_GPU(Avoid Nearest Surface=true, Avoid Oncoming Surfaces=true, Nearest Surface Avoidance Strength=25, Obstactle Avoidance Strength=25) {enabled=false}
   &SolveForcesAndVelocity(Clamp Velocity=true, Limit Acceleration=false, Speed Limit=100) {module=SolveForcesAndVelocity001}
   &CalculateAccurateVelocity()
   &ScaleColor(
    Scale Alpha=&Multiply_Float(
     A=&Multiply_Float(
      A=&Multiply_Float(
       A=1
       B=&FloatFromCurve(FloatCurve=DataInterface() {propertyValues {Curve {keys [<at 0 0>,<at 2 1>]}}}, CurveIndex=$Emitter.Age)
      )
      B=&FloatFromCurve(
       FloatCurve=DataInterface() {propertyValues {Curve {keys [
           {arriveTangent=9.0725, time=0, value=1}
           {arriveTangent=-0.4023, interpMode=RCIM_Cubic, leaveTangent=-0.4023, tangentMode=RCTM_User, time=0.75, value=0.85}
           {arriveTangent=-0.0405, interpMode=RCIM_Cubic, leaveTangent=-0.0405, tangentMode=RCTM_User, time=1, value=0}
          ]}}}
       Scale Curve=0.01
      )
     )
     B=/Niagara/DynamicInputs/Camera/CameraFade(Camera Query:NiagaraDataInterfaceCamera=?, Fade Offset:NiagaraFloat=500, Fade Distance:NiagaraFloat=500, Use Owner Position:NiagaraBool=true)
    )
    Scale RGB=<2 2 2>
    ScaleA=true
    ScaleRGB=true
   )
   &DynamicMaterialParameters(
    Index 0 Param 1=&MakeCustomFloatFromBool(
     Bool=$Particles.Is Active
     True Float=&RandomRangeFloat(Minimum=1.1, Maximum=2)
     False Float=&Multiply_Float(
      A=&Waveform(
       "Waveform [1]"=::NewEnumerator0
       "[1] Frequency"=&RandomRangeFloat(Minimum=0.25, Maximum=1)
       "[1] Amplitude Min/Max"=<0.5 1>
      )
      B=&RandomRangeFloat(Minimum=0.4, Maximum=0.8)
     )
    )
    Index 0 Param 2=&Multiply_Float(
     A=&MakeCustomFloatFromBool(
      Bool=$Particles.Is Active
      True Float=&RandomRangeFloat(Minimum=1, Maximum=1)
      False Float=&RandomRangeFloat(Minimum=0.5, Maximum=1)
     )
     B=&RandomRangeFloat(Minimum=2, Maximum=3)
    )
    Index 0 Param 3=&RandomRangeFloat(Minimum=1, Maximum=2)
    Index 1 Param 1=200
    Index 1 Param 2=&RandomRangeFloat(Minimum=0, Maximum=0.15)
    Module.Param0WriteEnabled=true
    Module.Param0WriteEnabled1=true
    Module.Param1WriteEnabled=true
    Module.Param1WriteEnabled1=true
    Module.Param2WriteEnabled=true
    Module.Param2WriteEnabled1=true
    Module.Param3WriteEnabled=true
    Module.Param3WriteEnabled1=true
    Write Parameter Index 1=true
   )
   &ScaleMeshSize(Scale Factor=&MakeVector(
     X=1
     Y=1
     Z=&FloatFromCurve(
      FloatCurve=DataInterface() {propertyValues {Curve {keys [<at 0 1>,<at 0.22 2>,<at 0.25 1>]}}}
      Scale Curve=&Waveform("Waveform [1]"=::NewEnumerator0, "[1] Frequency"=3, "[1] Amplitude Min/Max"=<1 1.5>)
     )
    )) {enabled=false}
   &self:ScratchModule_02(ZDampening=1) {module=ScratchModule_02}
   &KillParticlesInVolume(Invert Volume=true, Sphere Radius=8000) {enabled=false}
   &self:ScratchModule_04() {module=ScratchModule_04}
   /SetVariables(Particles.Linear Color=&MultiplyLinearColor(Linear Color=$Particles.Color, Scale Factor=<1.5 1.5 1.5 3>)) {isSetParameters=true, module=SetVariables_#5}
  ]
  rendererClasses [&NiagaraMeshRendererProperties, &NiagaraSpriteRendererProperties]
  renderers [
   {rendererClass=&NiagaraMeshRendererProperties, rendererIndex=0}
   {rendererClass=&NiagaraSpriteRendererProperties, rendererIndex=1}
  ]
 }
 Butterflies001 {
  simTarget=GPUComputeSim
  emitterUpdateScript [
   /SetVariables() {isSetParameters=true, module=SetVariables_#3}
   /SetVariables(
    Emitter.CameraForward:Vector3f=&ReturnCameraVector(Camera Property=::NewEnumerator1, Coordinate Space=::Simulation, CameraQuery=$Emitter.CameraRead)
    Emitter.CameraPosition:Vector3f=&ReturnCameraVector(Camera Property=::NewEnumerator0, Coordinate Space=::World, CameraQuery=$Emitter.CameraRead)
    Emitter.CameraRight:Vector3f=&ReturnCameraVector(Camera Property=::NewEnumerator3, Coordinate Space=::Simulation, CameraQuery=$Emitter.CameraRead)
    Emitter.CameraUp:Vector3f=&ReturnCameraVector(Camera Property=::NewEnumerator2, Coordinate Space=::Simulation, CameraQuery=$Emitter.CameraRead)
   ) {isSetParameters=true, module=SetVariables_#4}
   &EmitterState(
    Enable Distance Culling=true
    Life Cycle Mode=::NewEnumerator1
    Loop Behavior=::NewEnumerator1
    Loop Duration Mode=::NewEnumerator1
    Max Distance Response=::SleepAndClearParticles
    Module.bApplySpawnCountScaleByDistance=true
    Module.bUseMaxDistance=true
    Module.bUseMinDistance=false
    Reset Age On Awaken=true
    Scalability Mode=::NewEnumerator1
    Scale Spawn Count=true
    Spawn Count Scale By Distance=DataInterface() {propertyValues {Curve {keys [<at 0.5 1>,<at 0.55 1>,<at 1 1>]}}}
    UseLoopDelay=false
   )
   &SpawnPerFrame(
    Spawn:NiagaraBool=&SetBoolByFloatComparison(A=$Emitter.Age, B=1, Comparison Type=::NewEnumerator6)
    Spawn Count:NiagaraInt32=&Min_Int(
     A:NiagaraInt32=500
     B:NiagaraInt32=&Subtract_Integer(A:NiagaraInt32=50000, B:NiagaraInt32=$Engine.Emitter.NumParticles)
    )
    Use Spawn Probability=false
   ) {enabled=false}
   &SpawnRate(SpawnRate=25, Use Spawn Probability=false)
  ]
  particleSpawnScript [
   &InitializeParticle(
    Color=&ColorFromCurve(
     ColorCurve=DataInterface() {propertyValues {
       AlphaCurve {keys [<at 0 1 RCIM_Constant>,<at 1 1 RCIM_Constant>]}
       BlueCurve {keys [<at 0.0989 0.9911 RCIM_Constant>,<at 0.3996 0.8153 RCIM_Constant>,<at 0.6764 0.3907>]}
       GreenCurve {keys [<at 0.0989 0.9447 RCIM_Constant>,<at 0.3996 0.8906 RCIM_Constant>,<at 0.6764 0.8021>]}
       RedCurve {keys [<at 0.0989 0.0929 RCIM_Constant>,<at 0.3996 0.0009 RCIM_Constant>,<at 0.6764 0.0008>]}
      }}
     CurveIndex=&RandomRangeFloat(Minimum=0, Maximum=1)
    )
    Lifetime=&RandomRangeFloat(Minimum=3, Maximum=5)
    Mesh Scale Mode=::NewEnumerator1
    Mesh Uniform Scale Max=4
    Sprite Rotation Mode=::NewEnumerator2
    Sprite Size Mode=::NewEnumerator3
    Uniform Sprite Size=80
    UsePositionOffset=false
    Write Material Random=true
   ) {module=InitializeParticle001}
   &ShapeLocation(MaskSpawnGroup=false, Offset Mode=::NewEnumerator0, Override Seed=false, Sphere Radius=1000, UseOwnerScale=true)
   &PlaceParticlesOnDepthBuffer_GPU(Max Z Depth=10000, Module.Use Max Z Depth=true) {enabled=false}
   &SampleGBuffer(Kill Beyond Depth=true, Max Z Depth=10000, Snap To Depth:ENiagaraGBufferSnapToDepth=::NewEnumerator1, Write To Color=false) {enabled=false}
   &MoveToNearestDistanceFieldSurface_GPU(Kill by Distance=8000, Radius=500, bKillByDistance=false) {enabled=false}
   &InitialMeshOrientation(
    Mesh Axis To Orient=&RandomRangeVector(Minimum=<0 0 0>, Maximum=<1 1 1>)
    Mesh Orientation Mode=::NewEnumerator2
    Orientation Vector:NiagaraPosition=$Output.MoveToNearestDistanceFieldSurface_GPU.SurfacePositionWithoutRadiusOffset
    Rotation=&RandomRangeVector(Minimum=<0 0 0>, Maximum=<0 0 1>)
    Rotation Coordinate Space=::NewEnumerator2
    Use Rotation Vector=true
   )
   &CameraOffset(Camera Offset Amount=25)
   &KillParticles(Kill Particles=&SetBoolByFloatComparison(
     A=&VectorLength(Vector=&Subtract_Vector(A=$Particles.InitialPos, B=Linked() {linkedVariable {name=Particles.Position, type=NiagaraPosition}}))
     B=8000
    )) {enabled=false, module=KillParticles001}
   &KillParticles(Kill Particles=&Boolean_NotOperation(Bool=$Output.PlaceParticlesOnDepthBuffer_GPU.ValidSample)) {enabled=false, module=KillParticles002}
   &FN_Spawn_GBuffer_Color_WithinRangeV2(HueMax=110, HueMin=90) {enabled=false}
   &CameraOffset() {module=CameraOffset001}
   &AddVelocity(Module.Use Falloff Distance=false, Override Seed=false, Velocity Mode=::NewEnumerator1)
   &KillParticlesInVolume(Invert Volume=true, Sphere Radius=5000) {module=KillParticlesInVolume001}
  ]
  particleUpdateScript [
   &ParticleState()
   &ScaleSpriteSize()
   &CurlNoiseForce(
    Curl Noise Cone Mask Angle=90
    Mask Curl Noise=true
    Noise Frequency=&RandomRangeFloat(Minimum=0.85, Maximum=1)
    Noise Strength=2500
    Pan Noise=false
    Pan Noise Field=<0 0 0.1>
   )
   &ScaleMeshSize(Scale Factor=&MakeVector(
     X=2
     Y=2
     Z=&Waveform(
      Global Amplitude Scale=2
      "Waveform [1]"=::NewEnumerator0
      "[1] Phase Scale"=&RandomRangeFloat(Minimum=1, Maximum=3)
      "[1] Frequency"=1
      "[1] Amplitude Min/Max"=<1 4>
     )
    )) {module=ScaleMeshSize001}
   &AvoidDistanceFieldSurfaces_GPU(
    Avoid Nearest Surface=true
    Avoid Oncoming Surfaces=true
    Falloff Distance=500
    Max Trace Distance=500
    Nearest Surface Avoidance Strength=25
    Obstactle Avoidance Strength=25
   )
   &SolveForcesAndVelocity(Clamp Velocity=true, Limit Acceleration=false, Speed Limit=100) {module=SolveForcesAndVelocity001}
   &ScaleColor(
    Scale Alpha=&Multiply_Float(
     A=&Multiply_Float(
      A=&Multiply_Float(
       A=1
       B=&FloatFromCurve(FloatCurve=DataInterface() {propertyValues {Curve {
           defaultValue=0
           keys [
            {interpMode=RCIM_Cubic, tangentMode=RCTM_User, time=0, value=0}
            {interpMode=RCIM_Cubic, tangentMode=RCTM_User, time=0.5, value=1}
            {tangentMode=RCTM_User, time=1, value=0}
           ]
          }}})
      )
      B=&FloatFromCurve(
       FloatCurve=DataInterface() {propertyValues {Curve {keys [
           {arriveTangent=9.0725, time=0, value=1}
           {arriveTangent=-0.4023, interpMode=RCIM_Cubic, leaveTangent=-0.4023, tangentMode=RCTM_User, time=0.75, value=0.85}
           {arriveTangent=-0.0405, interpMode=RCIM_Cubic, leaveTangent=-0.0405, tangentMode=RCTM_User, time=1, value=0}
          ]}}}
       Scale Curve=0.01
      )
     )
     B=&CameraFade(Camera Query:NiagaraDataInterfaceCamera=?, Fade Offset:NiagaraFloat=250, Fade Distance:NiagaraFloat=500, Use Owner Position:NiagaraBool=true)
    )
    Scale RGB=<2 2 2>
    ScaleA=true
    ScaleRGB=true
   )
   &DynamicMaterialParameters(
    Index 0 Param 1=&MakeCustomFloatFromBool(
     Bool=$Particles.Is Active
     True Float=&RandomRangeFloat(Minimum=1.1, Maximum=3)
     False Float=&Multiply_Float(
      A=&Waveform(
       "Waveform [1]"=::NewEnumerator0
       "[1] Frequency"=&RandomRangeFloat(Minimum=0.25, Maximum=1)
       "[1] Amplitude Min/Max"=<0.5 1>
      )
      B=&RandomRangeFloat(Minimum=0.4, Maximum=0.8)
     )
    )
    Index 0 Param 2=2
    Index 0 Param 3=&RandomRangeFloat(Minimum=1, Maximum=2)
    Index 1 Param 1=200
    Index 1 Param 2=&RandomRangeFloat(Minimum=0, Maximum=0.15)
    Module.Param0WriteEnabled=true
    Module.Param0WriteEnabled1=true
    Module.Param1WriteEnabled=true
    Module.Param1WriteEnabled1=true
    Module.Param2WriteEnabled=true
    Module.Param2WriteEnabled1=true
    Module.Param3WriteEnabled=true
    Module.Param3WriteEnabled1=true
    Write Parameter Index 1=true
   )
   &ScaleMeshSize(Scale Factor=&MakeVector(
     X=1
     Y=1
     Z=&FloatFromCurve(
      FloatCurve=DataInterface() {propertyValues {Curve {keys [<at 0 1>,<at 0.22 2>,<at 0.25 1>]}}}
      Scale Curve=&Waveform("Waveform [1]"=::NewEnumerator0, "[1] Frequency"=3, "[1] Amplitude Min/Max"=<1 1.5>)
     )
    )) {enabled=false}
   &self:ScratchModule_02(ZDampening=1) {module=ScratchModule_02}
   &KillParticlesInVolume(Invert Volume=true, Sphere Radius=8000) {enabled=false}
   &self:ScratchModule_04() {module=ScratchModule_04}
   /SetVariables(Particles.Linear Color=&MultiplyLinearColor(Linear Color=$Particles.Color, Scale Factor=<3 3 3 3>)) {isSetParameters=true, module=SetVariables_#5}
  ]
  rendererClasses [&NiagaraMeshRendererProperties, &NiagaraSpriteRendererProperties]
  renderers [
   {rendererClass=&NiagaraMeshRendererProperties, rendererIndex=0}
   {rendererClass=&NiagaraSpriteRendererProperties, rendererIndex=1}
  ]
 }
 Fog {
  bEnabled=false
  emitterUpdateScript [
   &EmitterState(
    Enable Distance Culling=true
    MaxDistance=8000
    Module.bApplySpawnCountScaleByDistance=true
    Module.bUseMaxDistance=true
    Module.bUseMinDistance=false
    Scalability Mode=::NewEnumerator1
    Scale Spawn Count=true
    Spawn Count Scale By Distance=DataInterface() {propertyValues {Curve {keys [<at 0 1>,<at 0.1 1>,<at 0.75 1>,<at 1 0>]}}}
   )
   &SpawnRate(SpawnRate=15, Use Spawn Probability=false)
  ]
  particleSpawnScript [
   &InitializeParticle(
    Color=<rgba 0.4397 0.956 0.9823 1>
    Lifetime Max=6
    Lifetime Min=5
    Lifetime Mode=::NewEnumerator1
    Sprite Rotation Angle=&RandomRangeFloat(Minimum=-25, Maximum=25)
    Sprite Rotation Mode=::NewEnumerator0
    Sprite Size Mode=::NewEnumerator1
    Sprite UV Mode=::NewEnumerator1
    Uniform Sprite Size Max=1600
    Uniform Sprite Size Min=800
    UsePositionOffset=false
    Write Material Random=true
   )
   &ShapeLocation(
    Handle Radius=500
    Large Radius=1500
    MaskSpawnGroup=false
    Non Uniform Scale=<1 1 0.2>
    Offset=<0 0 -500>
    Offset Mode=::NewEnumerator0
    Override Seed=false
    Shape Primitive=::NewEnumerator3
    UseOwnerScale=true
   ) {module=ShapeLocation001}
   &AddVelocity(Velocity=<0 -2 0>)
  ]
  particleUpdateScript [
   &ParticleState()
   &ScaleSpriteSize(
    Scale Sprite Size Mode=::NewEnumerator0
    Uniform Scale Factor=&Multiply_Float(
     A=&FloatFromCurve(FloatCurve=DataInterface() {propertyValues {Curve {
         defaultValue=1
         keys [
          {interpMode=RCIM_Cubic, tangentMode=RCTM_User, time=0, value=0.7}
          {interpMode=RCIM_Cubic, tangentMode=RCTM_User, time=0.5, value=1}
          {tangentMode=RCTM_User, time=1, value=0.5}
         ]
        }}})
     B=&CameraFade(
      Camera Query:NiagaraDataInterfaceCamera=DataInterface()
      Fade Offset:NiagaraFloat=100
      Fade Distance:NiagaraFloat=500
      Use Owner Position:NiagaraBool=true
     )
    )
   ) {module=ScaleSpriteSize001}
   &ScaleColor(
    Scale Alpha=&Multiply_Float(
     A=&FloatFromCurve(
      FloatCurve=DataInterface() {propertyValues {Curve {keys [
          {interpMode=RCIM_Cubic, tangentMode=RCTM_User, time=0, value=0}
          {interpMode=RCIM_Cubic, tangentMode=RCTM_User, time=0.5, value=1}
          {tangentMode=RCTM_User, time=1, value=0}
         ]}}}
      Scale Curve=0.1
     )
     B=&CameraFade(
      Camera Query:NiagaraDataInterfaceCamera=DataInterface()
      Fade Offset:NiagaraFloat=500
      Fade Distance:NiagaraFloat=500
      Use Owner Position:NiagaraBool=true
     )
    )
    ScaleA=true
    ScaleRGB=true
   )
   &Drag(
    Drag=&FloatFromCurve(
     FloatCurve=DataInterface() {propertyValues {Curve {keys [
         {arriveTangent=1.805, interpMode=RCIM_Cubic, leaveTangent=1.805, tangentMode=RCTM_User, time=0, value=0}
         {arriveTangent=0.2851, interpMode=RCIM_Cubic, leaveTangent=0.2851, tangentMode=RCTM_User, time=1, value=1}
        ]}}}
     Scale Curve=3
    )
    Use Linear Drag=true
    Use Rotational Drag=true
   )
   &GravityForce(Gravity=&Multiply_VectorByFloat(Vector=<0 0 0.25>, Float=1))
   &SolveForcesAndVelocity(Clamp Velocity=false, Limit Acceleration=false)
   &SubUVAnimation(
    Frames Per Second:NiagaraInt32=&RandomRangeInt(Minimum=-6, Maximum=-8)
    Playback Mode:ENiagara_FPSPlayrate=::NewEnumerator1
    Random Start Frame=true
    "Sprite Renderer "=DataInterface() {propertyValues {SpriteRenderer=&self:Fog_0.NiagaraSpriteRendererProperties_0}}
    Start Frame Offset:NiagaraInt32=&RandomRangeInt(Minimum=0, Maximum=16)
    Start Frame Range Override=&RandomRangeInt(Minimum=0, Maximum=0)
    SubUV Animation Mode:ENiagara_SubUVLookupModeV2=::NewEnumerator3
    UseEndFrame=true
    UsePlayRateScale=false
    UseStartFrame=true
   )
   &DynamicMaterialParameters(
    Index 0 Param 1=&FloatFromCurve(FloatCurve=DataInterface() {propertyValues {Curve {keys [
         {interpMode=RCIM_Cubic, tangentMode=RCTM_User, time=0, value=1}
         {interpMode=RCIM_Cubic, tangentMode=RCTM_User, time=0.1, value=0}
         {arriveTangent=3.5426, interpMode=RCIM_Cubic, leaveTangent=3.5426, tangentMode=RCTM_User, time=1, value=1}
        ]}}})
    Index 0 Param 2=&FloatFromCurve(FloatCurve=DataInterface() {propertyValues {Curve {keys [
         {interpMode=RCIM_Cubic, tangentMode=RCTM_User, time=0, value=1}
         {interpMode=RCIM_Cubic, tangentMode=RCTM_User, time=0.25, value=0.5}
         {interpMode=RCIM_Cubic, tangentMode=RCTM_User, time=1, value=1}
        ]}}})
    Index 0 Param 3=250
    Module.Param0WriteEnabled=true
    Module.Param1WriteEnabled=true
    Module.Param2WriteEnabled=true
    Module.Param3WriteEnabled=true
   ) {module=DynamicMaterialParameters001}
   &CameraOffset(Camera Offset Amount=25)
  ]
  rendererClasses [&NiagaraSpriteRendererProperties]
  renderers [{rendererClass=&NiagaraSpriteRendererProperties}]
 }
 OnTree {
  simTarget=GPUComputeSim
  emitterUpdateScript [
   &EmitterState()
   &SpawnRate(
    SpawnRate=&Multiply_Float(
     A=500
     B=&FloatFromCurve(
      FloatCurve=DataInterface() {propertyValues {Curve {keys [<at 0 1>,<at 0.2 1>,<at 0.8 1>,<at 1 1>]}}}
      CurveIndex=Linked() {linkedVariable {name=NPC.FortniteNPC.TimeOfDay_0-1, type=NiagaraFloat}}
     )
    )
    Use Spawn Probability=false
   )
  ]
  particleSpawnScript [
   &InitializeParticle(
    Color Maximum=$User.BaseColor2
    Color Minimum=$User.BaseColor
    Color Mode=::NewEnumerator2
    Lifetime=&UniformRangedFloat(Minimum=3, Maximum=5)
    Mass Mode=::NewEnumerator1
    Position=&SimulationPosition()
    Position Mode=::NewEnumerator1
    Sprite Size Mode=::NewEnumerator1
    Uniform Sprite Size Max=8
    UsePositionOffset=false
    Write Material Random=true
   ) {module=InitializeParticle001}
   &AddVelocity(
    Cone Angle=90
    Cone Axis=<0 0 1>
    Module.UseConeFalloff=false
    Module.UseVelDistribution=false
    Override Seed=false
    Velocity Mode=::NewEnumerator2
    Velocity Speed=&RandomRangeFloat(Minimum=1, Maximum=250)
   )
   &InitialMeshOrientation(Mesh Axis To Orient=<0 0 1>, Mesh Orientation Mode=::NewEnumerator2, Orientation Vector=<0 0 1>, Use Rotation Vector=false)
   &ShapeLocation(
    Hemisphere Distribution=<1 0.7>
    MaskSpawnGroup=false
    Offset=<0 0 -500>
    Offset Mode=::NewEnumerator0
    Override Seed=false
    Spawn Group Mask=1
    Sphere Radius=4000
    UseOwnerScale=true
   ) {module=ShapeLocation001}
   &KillParticlesInVolume(Sphere Radius=1500) {enabled=false}
  ]
  particleUpdateScript [
   &ParticleState()
   &RegionMask(Inner Float=1.75, Region Radius:NiagaraFloat=$User.Region_ParticleSize, UseOffset=true, UseRotation=false)
   &RegionMask(Region Radius:NiagaraFloat=$User.Region_ParticleSize, UseOffset=true, UseRotation=false) {module=RegionMask001}
   &ScaleSpriteSize(
    Uniform Curve Scale=&Multiply_Float(
     A=$Particles.RegionMask.Float
     B=&Multiply_Float(
      A=&Lerp_Float(
       A=0.25
       B=1
       Alpha=&CameraFade(
        Camera Query:NiagaraDataInterfaceCamera=?
        Fade Offset:NiagaraFloat=500
        Fade Distance:NiagaraFloat=2500
        Use Owner Position:NiagaraBool=true
       )
      )
      B=&Multiply_Float(
       A=0.5
       B=&Waveform(
        "Waveform [1]"=::NewEnumerator0
        "[1] Frequency"=&RandomRangeFloat(Minimum=0.5, Maximum=0.7)
        "[1] Amplitude Min/Max"=<0.5 1>
       )
      )
     )
    )
    Uniform Curve Sprite Scale=DataInterface() {propertyValues {Curve {keys [
        {interpMode=RCIM_Cubic, tangentMode=RCTM_User, time=0, value=0}
        {arriveTangent=1.1111, interpMode=RCIM_Cubic, leaveTangent=1.1111, time=0.2, value=1}
        {interpMode=RCIM_Cubic, tangentMode=RCTM_User, time=0.9, value=1}
        {tangentMode=RCTM_User, time=1, value=0}
       ]}}}
   )
   &ScaleColor(
    Scale Alpha=&Multiply_Float(
     A=&FloatFromCurve(
      FloatCurve=DataInterface() {propertyValues {Curve {keys [
          {interpMode=RCIM_Cubic, tangentMode=RCTM_User, time=0, value=0}
          {interpMode=RCIM_Cubic, tangentMode=RCTM_User, time=0.5, value=1}
          {tangentMode=RCTM_User, time=1, value=0}
         ]}}}
      Scale Curve=&Multiply_Float(
       A=1
       B=&FloatFromCurve(
        FloatCurve=DataInterface() {propertyValues {Curve {keys [<at 0 1>,<at 0.2 1>,<at 0.8 1>,<at 1 1>]}}}
        CurveIndex=Linked() {linkedVariable {name=NPC.FortniteNPC.TimeOfDay_0-1, type=NiagaraFloat}}
       )
      )
     )
     B=$Particles.RegionMask001.Float
    )
    Scale RGB=&MakeVectorFromLinearColorRGB(Color=&HueShiftLinearColor(
      Hue=&Multiply_Float(A=$Particles.RegionMask001.Float, B=0.025)
      Saturation=1
      Value=1
      Linear Color=$Particles.Initial.Color
     ))
    ScaleA=true
    ScaleRGB=true
   )
   /Niagara/Modules/Update/Forces/CurlNoiseForce(
    Noise Frequency:NiagaraFloat=&RandomRangeFloat(Minimum=0.002, Maximum=0.005)
    Noise Strength=&RandomRangeFloat(Minimum=250, Maximum=2500)
    Pan Noise=true
    Pan Noise Field=<0 1 1>
   )
   &GravityForce(Gravity=<0 0 980>)
   &WindForce(Wind Speed=<2500 0 0>)
   &Drag(Drag=0.1, Use Linear Drag=true, Use Rotational Drag=true)
   &SolveForcesAndVelocity(Clamp Velocity=true, Limit Acceleration=false, Speed Limit=50)
   &CalculateAccurateVelocity() {enabled=false}
   /SetVariables(Particles.NewSize:Vector2f=&Multiply_Vector2D_ByFloat(
     Float=&FN_LOD_Distance_Remap(Start Distance=10000, End Distance=20000, Minimum Output=1, Maximum Output=0, Clamp Output=true)
     Vector2D=&Multiply_Vector2D_ByFloat(Float=1, Vector2D=$Particles.SpriteSize)
    )) {isSetParameters=true}
   /SetVariables(Particles.Vector 2D:Vector2f=&Multiply_Vector2D(A=$Particles.NewSize, B=<0.5 0.5>)) {isSetParameters=true}
   /SetVariables(Particles.Linear Color=&MultiplyLinearColor(Linear Color=$Particles.Color, Scale Factor=<1 1 1 0.25>)) {isSetParameters=true}
   &CameraOffset(Camera Offset Amount=&RandomRangeFloat(Minimum=400, Maximum=600))
  ]
  rendererClasses [&NiagaraSpriteRendererProperties]
  renderers [
   {rendererClass=&NiagaraSpriteRendererProperties}
   {rendererClass=&NiagaraSpriteRendererProperties, rendererIndex=1}
  ]
 }
}
summary {
 systemName=NS_SpriteHavenCave_Ambient
 emitters [
  {
   bEnabled=false
   emitterName=Butterflies
   simTarget=GPUComputeSim
   rendererClasses [&NiagaraMeshRendererProperties, &NiagaraSpriteRendererProperties]
  }
  {emitterName=OnTree, simTarget=GPUComputeSim, rendererClasses [&NiagaraSpriteRendererProperties]}
  {
   emitterName=Butterflies001
   simTarget=GPUComputeSim
   rendererClasses [&NiagaraMeshRendererProperties, &NiagaraSpriteRendererProperties]
  }
  {bEnabled=false, emitterName=Fog, rendererClasses [&NiagaraSpriteRendererProperties]}
 ]
 userVariables [
  User.BaseColor {type=LinearColor, description="", defaultValue {struct=LinearColor, value=<rgba 0.07 0.5521 0.4926 1>}}
  User.BaseColor2 {type=LinearColor, description="", defaultValue {struct=LinearColor, value=<rgba 0.0155 0.4479 0.2365 1>}}
  User.BoxSize {type=Vector3f, description="", defaultValue {struct=Vector3f, value=<0 0 0>}}
  User.Radius Mult {type=NiagaraFloat, defaultValue=1, description=""}
  User.Region_ParticleSize {type=NiagaraFloat, defaultValue=5000, description=""}
  User.Spawn Mult {type=NiagaraFloat, defaultValue=1, description=""}
 ]
}
