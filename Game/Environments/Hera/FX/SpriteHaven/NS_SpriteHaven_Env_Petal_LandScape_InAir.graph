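# Niagara environment effect for SpriteHaven that drifts clouds of tinted petals through the air, high above the system origin.
# GPU emitter Petals has no renderer and acts as moving leaders, spawned in a user-sized sphere and driven by curl noise, wind and aerodynamic drag.
# Petals001 spawns from those particles and draws short-lived sprites, scaled by LOD distance, with distance culling.
# Petals002, a sprite-plus-mesh variant, and landscape sampling are disabled.
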NiagaraSystem /Game/Environments/Hera/FX/SpriteHaven/NS_SpriteHaven_Env_Petal_LandScape_InAir
dependencies {
 usedDataInterfaces [
  &NiagaraDataInterfaceCurve
  &NiagaraDataInterfaceCollisionQuery
  &NiagaraDataInterfaceParticleRead
 ]
 usedDynamicInputs []
 usedModules [&SystemState]
}
emitters {
 Petals {
  simTarget=GPUComputeSim
  emitterUpdateScript [
   &EmitterState(
    Enable Distance Culling=true
    Max Distance Response=::SleepAndClearParticles
    MaxDistance=150000
    Module.bUseMaxDistance=true
    Module.bUseMinDistance=false
    Reset Age On Awaken=true
    Scalability Mode=::NewEnumerator1
   )
   &SpawnRate(SpawnRate=150, Use Spawn Probability=false)
  ]
  particleSpawnScript [
   &InitializeParticle(
    Color=&HueShiftLinearColor(
     Hue=&RandomRangeFloat(Minimum=-0.05, Maximum=0.1)
     Saturation=&RandomRangeFloat(Minimum=0.5, Maximum=1)
     Value=2
     Linear Color=$User.UserColor
    )
    Lifetime=&RandomRangeFloat(Minimum=1, Maximum=5)
    Mass=3
    Mass Mode=::NewEnumerator1
    Mesh Scale Mode=::NewEnumerator1
    Mesh Uniform Scale Max=0.05
    Mesh Uniform Scale Min=0.04
    Sprite Rotation Mode=::NewEnumerator2
    Sprite Size Mode=::NewEnumerator1
    Uniform Sprite Size Max=25
    Uniform Sprite Size Min=15
    UsePositionOffset=false
    Write Material Random=true
   ) {module=InitializeParticle001}
   /SetVariables(Particles.Terrain Offset:NiagaraFloat=&RandomRangeFloat(Minimum=25, Maximum=1500)) {isSetParameters=true}
   &ShapeLocation(
    MaskSpawnGroup=false
    Offset=&MakeVector(
     X=&FloatFromCurve(FloatCurve=DataInterface() {propertyValues {Curve {keys [<at 0 -2>,<at 1 2>]}}}, CurveIndex=$System.NormalizedLoopAge, Scale Curve=250)
     Y=0
     Z=4000
    )
    Offset Mode=::NewEnumerator0
    Override Seed=false
    Sphere Radius=$User.SphereLocation
    UseOwnerScale=true
   )
   /Niagara/Modules/Landscape/SampleLandscape(Height Offset:NiagaraFloat=&RandomRangeFloat(Minimum=3000, Maximum=5000)) {enabled=false}
   &CameraOffset(Camera Offset Amount=15)
   &SubUVAnimation(SubUV Animation Mode:ENiagara_SubUVLookupMode=::NewEnumerator2) {module=SubUVAnimation001}
   /SetVariables(Particles.SubImageIndex=&RandomRangeFloat(Minimum=0, Maximum=63)) {isSetParameters=true, module=SetVariables_#1}
   &InitialMeshOrientation(Rotation Coordinate Space=::NewEnumerator2, Use Rotation Vector=true)
   &CurlNoiseForce(Bias Noise=true, Bias Noise Field:Vector3f=<0.25 0 0>, Noise Frequency=0.01, Noise Strength=80000, Pan Noise=true, Pan Noise Field=<1 0 0>)
   &ApplyInitialForces()
  ]
  particleUpdateScript [
   &ScaleColor(
    Scale Alpha=&FloatFromCurve(FloatCurve=DataInterface() {propertyValues {Curve {keys [<at 0.8 1>,<at 1 0>]}}}, Scale Curve=5)
    Scale RGB=&MakeVectorFromLinearColorRGB(Color=&MultiplyLinearColor(Linear Color=$Particles.Initial.Color, Scale Factor=&MakeVector4_FromFloat(Float=1)))
    ScaleA=true
    ScaleRGB=true
   )
   &CurlNoiseForce(Bias Noise=true, Bias Noise Field:Vector3f=<0.5 0 0>, Noise Frequency=5, Noise Strength=25000, Pan Noise=true, Pan Noise Field=<1 0 0>) {module=CurlNoiseForce001}
   &SubUVAnimation(End Frame=64, SubUV Animation Mode:ENiagara_SubUVLookupMode=::NewEnumerator1, SubUV Loop Count=4)
   &ParticleState(DeltaTime:NiagaraFloat=&MakeCustomFloatFromBool(
     Bool=&SetBoolByFloatComparison(A=&VectorLength(Vector=$Particles.Velocity), B=1)
     True Float=$Engine.DeltaTime
     False Float=0
    ))
   &ScaleMeshSize(Scale Factor=&VectorFromFloat(Value=&Multiply_Float(
      A=&FloatFromCurve(
       FloatCurve=DataInterface() {propertyValues {Curve {keys [<at 0.25 0.25>,<at 500 1>,<at 6000 1>,<at 8000 0>]}}}
       CurveIndex=$Engine.Owner.LODDistance
       Scale Curve=0.5
      )
      B=&FloatFromCurve(FloatCurve=DataInterface() {propertyValues {Curve {keys [<at 0.8 1>,<at 1 0>]}}}, Scale Curve=0.25)
     )))
   &ScaleSpriteSize(
    Scale Sprite Size Mode=::NewEnumerator0
    Uniform Scale Factor=&Multiply_Float(
     A=&FloatFromCurve(
      FloatCurve=DataInterface() {propertyValues {Curve {keys [<at 0 1>,<at 500 1>,<at 6000 1>,<at 8000 0>]}}}
      CurveIndex=$Engine.Owner.LODDistance
      Scale Curve=2
     )
     B=&FloatFromCurve(FloatCurve=DataInterface() {propertyValues {Curve {keys [<at 0 0>,<at 0.125 0.9418>,<at 0.8 1>,<at 1 0>]}}}, Scale Curve=0.25)
    )
   )
   &WindForce(
    Friction Mode:ENiagaraWindFrictionMode=::NewEnumerator0
    Offset By Time:Vector3f=<0 0 -10>
    Scale Min / Max:Vector2f=<1 1>
    Speed Min / Max:Vector2f=<0 250>
    Turbulence Frequency Mode:ENiagaraWindTurbulenceFrequencyMode=::NewEnumerator1
    Turbulence Scale Mode:ENiagaraWindTurbulenceContributionMode=::NewEnumerator3
    Use Depth Buffer=false
    Use Landscape=false
    Wind Speed=<500 0 0>
    Wind Speed Scale=&RandomRangeFloat(Minimum=0.5, Maximum=1)
   ) {module=WindForce001}
   &AerodynamicDrag(
    Aerodynamic Drag=0.1
    Aerodynamic Rotational Drag=0.1
    Drag Shape Curve=DataInterface() {propertyValues {Curve {keys [
        {interpMode=RCIM_Cubic, tangentMode=RCTM_User, time=0, value=0.1}
        {arriveTangent=2.0799, interpMode=RCIM_Cubic, leaveTangent=2.0799, tangentMode=RCTM_User, time=1, value=1}
       ]}}}
    Lift Shape Curve=DataInterface() {propertyValues {Curve {keys [
        {interpMode=RCIM_Cubic, tangentMode=RCTM_User, time=0, value=0}
        {interpMode=RCIM_Cubic, tangentMode=RCTM_User, time=0.5, value=1}
        {tangentMode=RCTM_User, time=1, value=0}
       ]}}}
    Pivot Offset Maximum:Vector3f=<1 1 1>
    Pivot Offset Minimum Length:Vector3f=<0 0 0>
    Rotation Strength:NiagaraFloat=&RandomRangeFloat(Minimum=15, Maximum=25)
    Use Linear Drag=true
    Use Rotational Drag=true
   )
   &SolveForcesAndVelocity(
    Clamp Velocity=true
    Limit Acceleration=false
    Speed Limit:NiagaraFloat=&RandomRangeFloat(Minimum=250, Maximum=500)
   ) {module=SolveForcesAndVelocity001}
   &AlignSpriteToMeshOrientation(
    Mesh Orientation Relative Sprite Alignment Vector=<1 0 0>
    Mesh Orientation Relative Sprite Facing Vector:Vector3f=<1 1 1>
    SetAlignment=true
    SetFacing=true
   )
   &DynamicMaterialParameters(
    Index 0 Param 1=&RandomRangeFloat(Minimum=0.02, Maximum=0.1)
    Index 0 Param 2=&RandomRangeFloat(Minimum=-0.05, Maximum=0.05)
    Index 0 Param 3=1
    Module.Param0WriteEnabled=true
    Module.Param1WriteEnabled=true
    Module.Param2WriteEnabled=true
    Module.Param3WriteEnabled=true
   )
   /SetVariables(Particles.SpriteColor=&MultiplyLinearColor(Linear Color=$Particles.Color, Scale Factor=&MakeVector4_FromFloat(Float=2))) {isSetParameters=true, module=SetVariables_#2}
  ]
  rendererClasses []
  renderers []
 }
 Petals001 {
  simTarget=GPUComputeSim
  emitterUpdateScript [
   &EmitterState(
    Enable Distance Culling=true
    Max Distance Response=::SleepAndClearParticles
    MaxDistance=150000
    Module.bUseMaxDistance=true
    Module.bUseMinDistance=false
    Reset Age On Awaken=true
    Scalability Mode=::NewEnumerator1
   )
   &SpawnParticlesFromOtherEmitter(
    Attribute Reader=DataInterface() {propertyValues {EmitterBinding {bindingMode=Other, emitterName=Petals}}}
    Spawn Rate=100
    Spawn Rate Per Particle Cap=10
    Use Maximum Age=false
    Use Minimum Age=false
   )
  ]
  particleSpawnScript [
   &InitializeParticle(
    Color=&HueShiftLinearColor(
     Hue=&RandomRangeFloat(Minimum=-0.05, Maximum=0.1)
     Saturation=&RandomRangeFloat(Minimum=0.5, Maximum=1)
     Value=2
     Linear Color=$User.UserColor
    )
    Lifetime=&RandomRangeFloat(Minimum=1, Maximum=3)
    Mass=3
    Mass Mode=::NewEnumerator1
    Mesh Scale Mode=::NewEnumerator1
    Mesh Uniform Scale Max=0.05
    Mesh Uniform Scale Min=0.04
    Sprite Rotation Mode=::NewEnumerator2
    Sprite Size Mode=::NewEnumerator1
    Uniform Sprite Size Max=2
    Uniform Sprite Size Min=1
    UsePositionOffset=false
    Write Material Random=true
   ) {module=InitializeParticle001}
   &CameraOffset(Camera Offset Amount=15)
   &SubUVAnimation(SubUV Animation Mode:ENiagara_SubUVLookupMode=::NewEnumerator2) {enabled=false, module=SubUVAnimation001}
   /SetVariables(Particles.SubImageIndex=&RandomRangeFloat(Minimum=0, Maximum=63)) {enabled=false, isSetParameters=true, module=SetVariables_#1}
   &InitialMeshOrientation(Rotation Coordinate Space=::NewEnumerator2, Use Rotation Vector=true)
   &CurlNoiseForce(Bias Noise=true, Bias Noise Field:Vector3f=<0.25 0 0>, Noise Frequency=0.001, Noise Strength=500, Pan Noise=true, Pan Noise Field=<1 0 0>)
   &ApplyInitialForces()
   &SampleParticlesFromOtherEmitter(Applied Velocity Scale=0.1, Velocity Sampling=::NewEnumerator0)
   &ShapeLocation(
    MaskSpawnGroup=false
    Non Uniform Scale=<0 0 0>
    Offset=&MakeVector(
     X=&FloatFromCurve(FloatCurve=DataInterface() {propertyValues {Curve {keys [<at 0 -1>,<at 1 2>]}}}, CurveIndex=$System.NormalizedLoopAge, Scale Curve=250)
     Y=0
     Z=0
    )
    Offset Mode=::NewEnumerator0
    Override Seed=false
    Sphere Radius=$User.SphereLocation
    UseOwnerScale=true
   )
  ]
  particleUpdateScript [
   &ScaleColor(
    Scale Alpha=&FloatFromCurve(FloatCurve=DataInterface() {propertyValues {Curve {keys [<at 0.8 1>,<at 1 0>]}}}, Scale Curve=0.4)
    Scale RGB=&MakeVectorFromLinearColorRGB(Color=&MultiplyLinearColor(Linear Color=$Particles.Initial.Color, Scale Factor=&MakeVector4_FromFloat(Float=4)))
    ScaleA=true
    ScaleRGB=true
   )
   &CurlNoiseForce(Bias Noise=true, Bias Noise Field:Vector3f=<0.5 0 0>, Noise Frequency=0.01, Noise Strength=25, Pan Noise=true, Pan Noise Field=<1 0 0>) {module=CurlNoiseForce001}
   &SubUVAnimation(End Frame=64, SubUV Animation Mode:ENiagara_SubUVLookupMode=::NewEnumerator1, SubUV Loop Count=0.25) {enabled=false}
   &ParticleState(DeltaTime:NiagaraFloat=&MakeCustomFloatFromBool(
     Bool=&SetBoolByFloatComparison(A=&VectorLength(Vector=$Particles.Velocity), B=1)
     True Float=$Engine.DeltaTime
     False Float=0
    ))
   &ScaleMeshSize(Scale Factor=&VectorFromFloat(Value=&Multiply_Float(
      A=&FloatFromCurve(
       FloatCurve=DataInterface() {propertyValues {Curve {keys [<at 0.25 0.25>,<at 500 1>,<at 6000 1>,<at 8000 0>]}}}
       CurveIndex=$Engine.Owner.LODDistance
       Scale Curve=0.5
      )
      B=&FloatFromCurve(FloatCurve=DataInterface() {propertyValues {Curve {keys [<at 0.8 1>,<at 1 0>]}}}, Scale Curve=0.25)
     )))
   &ScaleSpriteSize(
    Scale Sprite Size Mode=::NewEnumerator0
    Uniform Scale Factor=&Multiply_Float(
     A=&FloatFromCurve(
      FloatCurve=DataInterface() {propertyValues {Curve {keys [<at 0 0>,<at 1500 1>,<at 8000 1>,<at 10000 0>]}}}
      CurveIndex=$Engine.Owner.LODDistance
      Scale Curve=3
     )
     B=&FloatFromCurve(FloatCurve=DataInterface() {propertyValues {Curve {keys [
          {interpMode=RCIM_Cubic, tangentMode=RCTM_User, time=0, value=0}
          {interpMode=RCIM_Cubic, tangentMode=RCTM_User, time=0.25, value=0.9418}
          {interpMode=RCIM_Cubic, tangentMode=RCTM_User, time=0.5, value=1}
          {interpMode=RCIM_Cubic, tangentMode=RCTM_User, time=1, value=0}
         ]}}})
    )
   )
   &WindForce(
    Friction Mode:ENiagaraWindFrictionMode=::NewEnumerator0
    Offset By Time:Vector3f=<0 0 -10>
    Scale Min / Max:Vector2f=<1 0.2>
    Speed Min / Max:Vector2f=<0 250>
    Turbulence Frequency Mode:ENiagaraWindTurbulenceFrequencyMode=::NewEnumerator1
    Turbulence Scale Mode:ENiagaraWindTurbulenceContributionMode=::NewEnumerator3
    Use Depth Buffer=false
    Use Landscape=false
    Wind Speed Scale=&RandomRangeFloat(Minimum=0.5, Maximum=1)
   ) {module=WindForce001}
   &AerodynamicDrag(
    Aerodynamic Rotational Drag=0.1
    Drag Shape Curve=DataInterface() {propertyValues {Curve {keys [
        {interpMode=RCIM_Cubic, tangentMode=RCTM_User, time=0, value=0.1}
        {arriveTangent=2.0799, interpMode=RCIM_Cubic, leaveTangent=2.0799, tangentMode=RCTM_User, time=1, value=1}
       ]}}}
    Lift Shape Curve=DataInterface() {propertyValues {Curve {keys [
        {interpMode=RCIM_Cubic, tangentMode=RCTM_User, time=0, value=0}
        {interpMode=RCIM_Cubic, tangentMode=RCTM_User, time=0.5, value=1}
        {tangentMode=RCTM_User, time=1, value=0}
       ]}}}
    Pivot Offset Maximum:Vector3f=<1 1 1>
    Pivot Offset Minimum Length:Vector3f=<0 0 0>
    Rotation Strength:NiagaraFloat=&RandomRangeFloat(Minimum=0.2, Maximum=0.5)
    Use Linear Drag=true
    Use Rotational Drag=true
   )
   &SolveForcesAndVelocity(
    Clamp Velocity=true
    Limit Acceleration=false
    Speed Limit:NiagaraFloat=&RandomRangeFloat(Minimum=250, Maximum=500)
   ) {module=SolveForcesAndVelocity001}
   /SetVariables(Particles.SpriteColor=&MultiplyLinearColor(
     Linear Color=$Particles.Color
     Scale Factor=&MakeVector4(
      X=1
      Y=0
      Z=1
      W=&FloatFromCurve(FloatCurve=DataInterface() {propertyValues {Curve {
          defaultValue=1
          keys [
           {interpMode=RCIM_Cubic, tangentMode=RCTM_User, time=0, value=0}
           {interpMode=RCIM_Cubic, tangentMode=RCTM_User, time=0.062, value=0.9418}
           {tangentMode=RCTM_User, time=1, value=0}
          ]
         }}})
     )
    )) {enabled=false, isSetParameters=true, module=SetVariables_#2}
   &DynamicMaterialParameters(
    Index 0 Param 1=&RandomRangeFloat(Minimum=0.01, Maximum=0.2)
    Index 0 Param 2=&RandomRangeFloat(Minimum=0, Maximum=0.02)
    Index 0 Param 3=1
    Module.Param0WriteEnabled=true
    Module.Param1WriteEnabled=true
    Module.Param2WriteEnabled=true
    Module.Param3WriteEnabled=true
   )
   /SetVariables(Particles.NewSize:Vector2f=&Multiply_Vector2D_ByFloat(
     Float=&FN_LOD_Distance_Remap(Start Distance=10000, End Distance=25000, Minimum Output=1, Maximum Output=0, Clamp Output=true)
     Vector2D=&Multiply_Vector2D_ByFloat(Float=5, Vector2D=$Particles.SpriteSize)
    )) {isSetParameters=true}
   /SetVariables(Particles.Vector 2D:Vector2f=&Multiply_Vector2D(A=$Particles.NewSize, B=<1 1>)) {isSetParameters=true}
   /SetVariables(Particles.Linear Color=&MultiplyLinearColor(Linear Color=$Particles.Color, Scale Factor=<1 1 1 0.25>)) {isSetParameters=true}
  ]
  rendererClasses [&NiagaraSpriteRendererProperties]
  renderers [
   {rendererClass=&NiagaraSpriteRendererProperties}
   {rendererClass=&NiagaraSpriteRendererProperties, rendererIndex=1}
   {rendererClass=&NiagaraSpriteRendererProperties, rendererIndex=2}
  ]
 }
 Petals002 {
  bEnabled=false
  simTarget=GPUComputeSim
  emitterUpdateScript [
   &EmitterState(
    Enable Distance Culling=true
    Max Distance Response=::SleepAndClearParticles
    MaxDistance=150000
    Module.bUseMaxDistance=true
    Module.bUseMinDistance=false
    Reset Age On Awaken=true
    Scalability Mode=::NewEnumerator1
   )
   &SpawnParticlesFromOtherEmitter(
    Attribute Reader=DataInterface() {propertyValues {EmitterBinding {bindingMode=Other, emitterName=Petals}}}
    Spawn Rate=100
    Spawn Rate Per Particle Cap=10
    Use Maximum Age=false
    Use Minimum Age=false
   )
  ]
  particleSpawnScript [
   &InitializeParticle(
    Color=&HueShiftLinearColor(
     Hue=&RandomRangeFloat(Minimum=-0.05, Maximum=0.1)
     Saturation=&RandomRangeFloat(Minimum=0.5, Maximum=1)
     Value=2
     Linear Color=$User.UserColor
    )
    Lifetime=&RandomRangeFloat(Minimum=1, Maximum=3)
    Mass=3
    Mass Mode=::NewEnumerator1
    Mesh Scale Mode=::NewEnumerator1
    Mesh Uniform Scale Max=0.05
    Mesh Uniform Scale Min=0.04
    Sprite Rotation Mode=::NewEnumerator2
    Sprite Size Mode=::NewEnumerator1
    Uniform Sprite Size Max=5
    Uniform Sprite Size Min=2
    UsePositionOffset=false
    Write Material Random=true
   ) {module=InitializeParticle001}
   &CameraOffset(Camera Offset Amount=15)
   &SubUVAnimation(SubUV Animation Mode:ENiagara_SubUVLookupMode=::NewEnumerator2) {enabled=false, module=SubUVAnimation001}
   /SetVariables(Particles.SubImageIndex=&RandomRangeFloat(Minimum=0, Maximum=63)) {enabled=false, isSetParameters=true, module=SetVariables_#1}
   &InitialMeshOrientation(Rotation Coordinate Space=::NewEnumerator2, Use Rotation Vector=true)
   &CurlNoiseForce(Bias Noise=true, Bias Noise Field:Vector3f=<0.25 0 0>, Noise Frequency=0.001, Noise Strength=500, Pan Noise=true, Pan Noise Field=<1 0 0>)
   &ApplyInitialForces()
   &SampleParticlesFromOtherEmitter(Applied Velocity Scale=0.1, Velocity Sampling=::NewEnumerator0)
   &ShapeLocation(
    MaskSpawnGroup=false
    Non Uniform Scale=<0 0 0>
    Offset=&MakeVector(
     X=&FloatFromCurve(FloatCurve=DataInterface() {propertyValues {Curve {keys [<at 0 -1>,<at 1 2>]}}}, CurveIndex=$System.NormalizedLoopAge, Scale Curve=250)
     Y=0
     Z=0
    )
    Offset Mode=::NewEnumerator0
    Override Seed=false
    Sphere Radius=$User.SphereLocation
    UseOwnerScale=true
   )
  ]
  particleUpdateScript [
   &ScaleColor(
    Scale Alpha=&FloatFromCurve(FloatCurve=DataInterface() {propertyValues {Curve {keys [<at 0.8 1>,<at 1 0>]}}}, Scale Curve=0.4)
    Scale RGB=&MakeVectorFromLinearColorRGB(Color=&MultiplyLinearColor(Linear Color=$Particles.Initial.Color, Scale Factor=&MakeVector4_FromFloat(Float=4)))
    ScaleA=true
    ScaleRGB=true
   )
   &CurlNoiseForce(Bias Noise=true, Bias Noise Field:Vector3f=<0.5 0 0>, Noise Frequency=0.01, Noise Strength=25, Pan Noise=true, Pan Noise Field=<1 0 0>) {module=CurlNoiseForce001}
   &SubUVAnimation(End Frame=64, SubUV Animation Mode:ENiagara_SubUVLookupMode=::NewEnumerator1, SubUV Loop Count=0.25) {enabled=false}
   &ParticleState(DeltaTime:NiagaraFloat=&MakeCustomFloatFromBool(
     Bool=&SetBoolByFloatComparison(A=&VectorLength(Vector=$Particles.Velocity), B=1)
     True Float=$Engine.DeltaTime
     False Float=0
    ))
   &ScaleMeshSize(Scale Factor=&VectorFromFloat(Value=&Multiply_Float(
      A=&FloatFromCurve(
       FloatCurve=DataInterface() {propertyValues {Curve {keys [<at 0.25 0.25>,<at 500 1>,<at 6000 1>,<at 8000 0>]}}}
       CurveIndex=$Engine.Owner.LODDistance
       Scale Curve=0.5
      )
      B=&FloatFromCurve(FloatCurve=DataInterface() {propertyValues {Curve {keys [<at 0.8 1>,<at 1 0>]}}}, Scale Curve=0.25)
     )))
   &ScaleSpriteSize(
    Scale Sprite Size Mode=::NewEnumerator0
    Uniform Scale Factor=&Multiply_Float(
     A=&FloatFromCurve(
      FloatCurve=DataInterface() {propertyValues {Curve {keys [<at 0 0>,<at 1500 1>,<at 8000 1>,<at 10000 0>]}}}
      CurveIndex=$Engine.Owner.LODDistance
      Scale Curve=3
     )
     B=&FloatFromCurve(FloatCurve=DataInterface() {propertyValues {Curve {keys [
          {interpMode=RCIM_Cubic, tangentMode=RCTM_User, time=0, value=0}
          {interpMode=RCIM_Cubic, tangentMode=RCTM_User, time=0.25, value=0.9418}
          {interpMode=RCIM_Cubic, tangentMode=RCTM_User, time=0.5, value=1}
          {interpMode=RCIM_Cubic, tangentMode=RCTM_User, time=1, value=0}
         ]}}})
    )
   )
   &WindForce(
    Friction Mode:ENiagaraWindFrictionMode=::NewEnumerator0
    Offset By Time:Vector3f=<0 0 -10>
    Scale Min / Max:Vector2f=<1 0.2>
    Speed Min / Max:Vector2f=<0 250>
    Turbulence Frequency Mode:ENiagaraWindTurbulenceFrequencyMode=::NewEnumerator1
    Turbulence Scale Mode:ENiagaraWindTurbulenceContributionMode=::NewEnumerator3
    Use Depth Buffer=false
    Use Landscape=false
    Wind Speed Scale=&RandomRangeFloat(Minimum=0.5, Maximum=1)
   ) {module=WindForce001}
   &AerodynamicDrag(
    Aerodynamic Rotational Drag=0.1
    Drag Shape Curve=DataInterface() {propertyValues {Curve {keys [
        {interpMode=RCIM_Cubic, tangentMode=RCTM_User, time=0, value=0.1}
        {arriveTangent=2.0799, interpMode=RCIM_Cubic, leaveTangent=2.0799, tangentMode=RCTM_User, time=1, value=1}
       ]}}}
    Lift Shape Curve=DataInterface() {propertyValues {Curve {keys [
        {interpMode=RCIM_Cubic, tangentMode=RCTM_User, time=0, value=0}
        {interpMode=RCIM_Cubic, tangentMode=RCTM_User, time=0.5, value=1}
        {tangentMode=RCTM_User, time=1, value=0}
       ]}}}
    Pivot Offset Maximum:Vector3f=<1 1 1>
    Pivot Offset Minimum Length:Vector3f=<0 0 0>
    Rotation Strength:NiagaraFloat=&RandomRangeFloat(Minimum=0.2, Maximum=0.5)
    Use Linear Drag=true
    Use Rotational Drag=true
   )
   &SolveForcesAndVelocity(
    Clamp Velocity=true
    Limit Acceleration=false
    Speed Limit:NiagaraFloat=&RandomRangeFloat(Minimum=250, Maximum=500)
   ) {module=SolveForcesAndVelocity001}
   /SetVariables(Particles.SpriteColor=&MultiplyLinearColor(
     Linear Color=$Particles.Color
     Scale Factor=&MakeVector4(
      X=1
      Y=0
      Z=1
      W=&FloatFromCurve(FloatCurve=DataInterface() {propertyValues {Curve {
          defaultValue=1
          keys [
           {interpMode=RCIM_Cubic, tangentMode=RCTM_User, time=0, value=0}
           {interpMode=RCIM_Cubic, tangentMode=RCTM_User, time=0.062, value=0.9418}
           {tangentMode=RCTM_User, time=1, value=0}
          ]
         }}})
     )
    )) {enabled=false, isSetParameters=true, module=SetVariables_#2}
   &DynamicMaterialParameters(
    Index 0 Param 1=&RandomRangeFloat(Minimum=0.01, Maximum=0.2)
    Index 0 Param 2=&RandomRangeFloat(Minimum=0, Maximum=0.02)
    Index 0 Param 3=1
    Module.Param0WriteEnabled=true
    Module.Param1WriteEnabled=true
    Module.Param2WriteEnabled=true
    Module.Param3WriteEnabled=true
   )
  ]
  rendererClasses [&NiagaraSpriteRendererProperties, &NiagaraMeshRendererProperties]
  renderers [
   {rendererClass=&NiagaraSpriteRendererProperties}
   {rendererClass=&NiagaraMeshRendererProperties, rendererIndex=1}
   {rendererClass=&NiagaraSpriteRendererProperties, rendererIndex=2}
  ]
 }
}
summary {
 systemName=NS_SpriteHaven_Env_Petal_LandScape_InAir
 emitters [
  {emitterName=Petals, simTarget=GPUComputeSim, rendererClasses []}
  {emitterName=Petals001, simTarget=GPUComputeSim, rendererClasses [&NiagaraSpriteRendererProperties]}
  {
   bEnabled=false
   emitterName=Petals002
   simTarget=GPUComputeSim
   rendererClasses [&NiagaraSpriteRendererProperties, &NiagaraMeshRendererProperties]
  }
 ]
 userVariables [
  User.Alpha_AttractionStrength {type=NiagaraFloat, defaultValue=0, description=""}
  User.Landscape {
   type=NiagaraDataInterfaceLandscape
   description=""
   defaultValue {
    struct=NiagaraExt_VariableValue_DataInterface
    value {
     dataInterface=&self:NiagaraDataInterfaceLandscape_0
     dataInterfaceClass=&NiagaraDataInterfaceLandscape
    }
   }
  }
  User.Player Position {type=Vector3f, description="", defaultValue {struct=Vector3f, value=<0 0 0>}}
  User.SpawnRate {type=NiagaraFloat, defaultValue=2500, description=""}
  User.SphereLocation {type=NiagaraFloat, defaultValue=2000, description=""}
  User.UserColor {type=LinearColor, description="", defaultValue {struct=LinearColor, value=<rgba 0.8603 0 0.8714 1>}}
 ]
}
