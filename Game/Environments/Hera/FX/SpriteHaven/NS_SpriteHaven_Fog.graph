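# Niagara system for drifting, faintly pink fog in the Hera SpriteHaven environment, using one active GPU emitter with two sprite renderers and one mesh renderer.
# Particles spawn continuously in a wide, flat, owner-scaled box and run a landscape sample at spawn.
# Acceleration and drag drive a slow sideways, rising drift, while alpha fades in and out faintly over lifetime.
# An inverted ViewMask output scales particle size and opacity according to camera view.
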NiagaraSystem /Game/Environments/Hera/FX/SpriteHaven/NS_SpriteHaven_Fog
dependencies {
 usedDataInterfaces [&NiagaraDataInterfaceLandscape, &NiagaraDataInterfaceCurve, &NiagaraDataInterfaceColorCurve]
 usedDynamicInputs []
 usedModules [&SystemState]
}
emitters {
 Empty {
  simTarget=GPUComputeSim
  emitterUpdateScript [
   &EmitterState()
   &SpawnRate(SpawnRate=25, Use Spawn Probability=false)
  ]
  particleSpawnScript [
   &InitializeParticle(
    Color=<rgba 1 0.7604 0.9387 1>
    Lifetime=5
    Mesh Scale Mode=::NewEnumerator1
    Mesh Uniform Scale Max=12
    Mesh Uniform Scale Min=2
    Sprite Size Mode=::NewEnumerator1
    Uniform Sprite Size Max=1500
    Uniform Sprite Size Min=1000
    UsePositionOffset=false
    Write Material Random=true
   )
   &ShapeLocation(
    Box Size=<15000 15000 0>
    MaskSpawnGroup=false
    Offset=<0 0 25>
    Offset Mode=::NewEnumerator0
    Override Seed=false
    Shape Primitive=::NewEnumerator2
    Surface Only=false
    UseOwnerScale=true
   )
   /Niagara/Modules/Landscape/SampleLandscape()
  ]
  particleUpdateScript [
   &ParticleState()
   /Niagara/Modules/Update/Camera/ViewMask(Border Percent Max=80, Border Percent Min=10)
   &ScaleMeshSize(Scale Factor=&Multiply_VectorByFloat(Vector=<1 1 1>, Float=&OneMinusFloat(Float=$Output.ViewMask.ViewMask)))
   &ScaleSpriteSize(
    Uniform Curve Scale=&Multiply_Float(A=1, B=&OneMinusFloat(Float=$Output.ViewMask.ViewMask))
    Uniform Curve Sprite Scale=DataInterface() {propertyValues {Curve {keys [{interpMode=RCIM_Cubic, tangentMode=RCTM_Break, time=0, value=0}, {tangentMode=RCTM_User, time=1, value=1}]}}}
   )
   &ScaleColor(
    Scale Alpha=&FloatFromCurve(
     FloatCurve=DataInterface() {propertyValues {Curve {keys [
         {interpMode=RCIM_Cubic, tangentMode=RCTM_User, time=0, value=0}
         {interpMode=RCIM_Cubic, tangentMode=RCTM_User, time=0.5, value=1}
         {tangentMode=RCTM_User, time=1, value=0}
        ]}}}
     Scale Curve=&Multiply_Float(A=0.1, B=&OneMinusFloat(Float=$Output.ViewMask.ViewMask))
    )
    ScaleA=true
    ScaleRGB=true
   )
   &AccelerationForce(Acceleration=<150 0 25>)
   &Drag(Use Linear Drag=true, Use Rotational Drag=true)
   &SolveForcesAndVelocity(Clamp Velocity=false, Limit Acceleration=false)
  ]
  rendererClasses [&NiagaraSpriteRendererProperties, &NiagaraMeshRendererProperties]
  renderers [
   {rendererClass=&NiagaraSpriteRendererProperties}
   {rendererClass=&NiagaraMeshRendererProperties, rendererIndex=1}
   {rendererClass=&NiagaraSpriteRendererProperties, rendererIndex=2}
  ]
 }
 Empty001 {bEnabled=false, rendererClasses [], renderers []}
}
summary {
 systemName=NS_SpriteHaven_Fog
 emitters [
  {
   emitterName=Empty
   simTarget=GPUComputeSim
   rendererClasses [&NiagaraSpriteRendererProperties, &NiagaraMeshRendererProperties]
  }
  {bEnabled=false, emitterName=Empty001, rendererClasses []}
 ]
 userVariables []
}
