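# Niagara GPU system for the SpriteHaven environment that scatters small, color-varied frog meshes across the landscape.
# A short spawn burst fills a flattened sphere, then landscape sampling orients each frog to the ground normal.
# Camera-relative repulsion startles frogs, latching an Is Active flag that drives hop-like gravity and livelier material parameters.
# Velocity gates particle aging, while collision, waveform squash-and-stretch, end-of-life shrinking and kill volumes complete the behavior.
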NiagaraSystem /Game/Environments/Hera/FX/SpriteHaven/NS_SpriteHaven_Frog_OnLandScape
dependencies {
 usedDataInterfaces [
  &NiagaraDataInterfaceCamera
  &NiagaraDataInterfaceCurve
  &NiagaraDataInterfaceColorCurve
  &NiagaraDataInterfaceLandscape
  &NiagaraDataInterfaceArrayFloat
 ]
 usedDynamicInputs [/Game/Effects/Niagara/DynamicInputs/Math/FN_Fit]
 usedModules [
  &self:SystemSpawnScript.SystemScriptSource.SystemScriptGraph.NiagaraNodeAssignment_0.SetVariables_#1
  &SystemState
  &self:SystemSpawnScript.SystemScriptSource.SystemScriptGraph.NiagaraNodeAssignment_1.SetVariables_#2
 ]
}
emitters {Frogs {
  simTarget=GPUComputeSim
  emitterUpdateScript [
   /SetVariables() {isSetParameters=true}
   /SetVariables(
    Emitter.CameraForward:Vector3f=&ReturnCameraVector(Camera Property=::NewEnumerator1, Coordinate Space=::Simulation, CameraQuery=$Emitter.CameraRead)
    Emitter.CameraPosition:Vector3f=&ReturnCameraVector(Camera Property=::NewEnumerator0, Coordinate Space=::World, CameraQuery=$Emitter.CameraRead)
    Emitter.CameraRight:Vector3f=&ReturnCameraVector(Camera Property=::NewEnumerator3, Coordinate Space=::Simulation, CameraQuery=$Emitter.CameraRead)
    Emitter.CameraUp:Vector3f=&ReturnCameraVector(Camera Property=::NewEnumerator2, Coordinate Space=::Simulation, CameraQuery=$Emitter.CameraRead)
   ) {isSetParameters=true}
   &EmitterState(Life Cycle Mode=::NewEnumerator1, Loop Behavior=::NewEnumerator1, Loop Duration Mode=::NewEnumerator1, UseLoopDelay=false)
   &SpawnRate(
    SpawnRate=&FloatFromCurve(FloatCurve=DataInterface() {propertyValues {Curve {keys [<at 0.1 1>,<at 0.2 0>]}}}, CurveIndex=$Emitter.Age, Scale Curve=375)
    Use Spawn Probability=false
   )
  ]
  particleSpawnScript [
   &InitializeParticle(
    AdjustAlpha=false
    AdjustHue=true
    AdjustSaturation=true
    AdjustValue=true
    Color=&ColorFromCurve(
     ColorCurve=DataInterface() {propertyValues {
       AlphaCurve {keys [<at 0 1 RCIM_Constant>,<at 1 1 RCIM_Constant>]}
       BlueCurve {keys [<at 0 0 RCIM_Constant>,<at 0.0989 0.2083 RCIM_Constant>,<at 0.3996 0.625 RCIM_Constant>,<at 0.6957 0 RCIM_Constant>,<at 0.7982 0 RCIM_Constant>]}
       GreenCurve {keys [<at 0 0.2865 RCIM_Constant>,<at 0.0989 0.1601 RCIM_Constant>,<at 0.3996 0.4218 RCIM_Constant>,<at 0.6957 0.4688 RCIM_Constant>,<at 0.7982 1 RCIM_Constant>]}
       RedCurve {keys [<at 0 0.0275 RCIM_Constant>,<at 0.0989 0.0235 RCIM_Constant>,<at 0.3996 0.0276 RCIM_Constant>,<at 0.6957 0.2161 RCIM_Constant>,<at 0.7982 0.0496 RCIM_Constant>]}
      }}
     CurveIndex=&RandomRangeFloat(Minimum=0, Maximum=1)
    )
    Color Mode=::NewEnumerator3
    Hue Shift Range:Vector2f=<0 0.025>
    Lifetime=&RandomRangeFloat(Minimum=3, Maximum=5)
    Mesh Scale Mode=::NewEnumerator1
    Mesh Uniform Scale Max=0.15
    Mesh Uniform Scale Min=0.15
    Saturation Range:Vector2f=<0.6 0.8>
    Sprite Rotation Mode=::NewEnumerator2
    UsePositionOffset=false
    Value Range:Vector2f=<0.2 0.4>
    Write Material Random=true
   ) {module=InitializeParticle001}
   &ShapeLocation(
    MaskSpawnGroup=false
    Non Uniform Scale=<1 1 0.02>
    Offset Mode=::NewEnumerator0
    Override Seed=false
    Sphere Radius=1000
    UseOwnerScale=true
   )
   /Niagara/Modules/Landscape/SampleLandscape(Facing Axis:Vector3f=<0 0 1>, Orient Mesh To Normal=true)
   &CameraOffset(Camera Offset Amount=10)
   /SetVariables() {isSetParameters=true}
   &InitialMeshOrientation(
    Mesh Axis To Orient=<0 0 1>
    Mesh Orientation Mode=::NewEnumerator2
    Orientation Vector=$Output.SampleLandscape.Normal
    Rotation=&RandomRangeVector(Minimum=<0 0 0>, Maximum=<0 0 1>)
    Rotation Coordinate Space=::NewEnumerator2
    Use Rotation Vector=true
   )
   &CameraOffset(Camera Offset Amount=25) {module=CameraOffset001}
   &KillParticlesInVolume(Box Size:Vector3f=<2500 2500 250>, Invert Volume=true, Kill Shape=::NewEnumerator1) {module=KillParticlesInVolume001}
  ]
  particleUpdateScript [
   &ParticleState(DeltaTime:NiagaraFloat=&MakeCustomFloatFromBool(
     Bool=&SetBoolByFloatComparison(A=&VectorLength(Vector=$Particles.Velocity), B=1)
     True Float=$Engine.DeltaTime
     False Float=0
    ))
   /Niagara/Modules/Update/Forces/PointAttractionForce(
    Attraction Radius:NiagaraFloat=&RandomRangeFloat(Minimum=200, Maximum=300)
    AttractionStrength:NiagaraFloat=&Multiply_Float(
     A=&FloatFromCurve(
      FloatCurve=DataInterface() {propertyValues {Curve {keys [<at 0 1>,<at 0.25 1>]}}}
      Scale Curve=&RandomRangeFloat(Minimum=-1, Maximum=-5)
     )
     B=1
    )
    AttractorPosition:Vector3f=&Add_Vector(
     A=$Emitter.CameraPosition
     B=&Add_Vector(
      A=&Add_Vector(
       A=&Multiply_VectorByFloat(Vector=<0 0 1>, Float=-300)
       B=&Multiply_VectorByFloat(Vector=$Emitter.CameraForward, Float=-250)
      )
      B=&Multiply_VectorByFloat(Vector=$Emitter.CameraRight, Float=0)
     )
    )
    Falloff Exponent=1
    Module.Kill Within Radius=false
    Module.Use Falloff=false
   )
   /SetVariables(Particles.Normalized Velocity:Vector3f=&NormalizeVector(Vector To Normalize=$Particles.Velocity)) {isSetParameters=true}
   /SetVariables(Particles.Is Active:NiagaraBool=&Boolean_OrOperation(
     Bool1=$Particles.Is Active
     Bool2=&SetBoolByFloatComparison(A=&VectorLength(Vector=$Particles.Normalized Velocity), B=0.1)
    )) {isSetParameters=true}
   &PointAttractionForce(
    Attraction Radius=500
    AttractionStrength:NiagaraFloat=&Multiply_Float(
     A=&MakeCustomFloatFromBool(
      Bool=$Particles.Is Active
      True Float=0
      False Float=&FloatFromCurve(FloatCurve=DataInterface() {propertyValues {Curve {keys [<at 0 1>,<at 1 0>]}}}, Scale Curve=-8)
     )
     B=1
    )
    AttractorPosition:Vector3f=&self:ScratchDynamicInput(
     CameraPosition:NiagaraPosition=Linked() {linkedVariable {name=Emitter.CameraPosition, type=Vector3f}}
     CameraForward:Vector3f=$Emitter.CameraForward
     HeightOffset:NiagaraFloat=&RandomRangeFloat(Minimum=-50, Maximum=-150)
     CameraForwardOffset:NiagaraFloat=&RandomRangeFloat(Minimum=200, Maximum=300)
    )
    Falloff Exponent=1
    Module.Kill Within Radius=false
    Module.Use Falloff=false
   ) {module=PointAttractionForce001}
   &GravityForce(Gravity=&Multiply_VectorByFloat(
     Vector=<0 0 3>
     Float=&MakeCustomFloatFromBool(
      Bool=$Particles.Is Active
      True Float=&FloatFromCurve(FloatCurve=DataInterface() {propertyValues {Curve {keys [<at 0 1>,<at 0.1 -1>]}}}, Scale Curve=300)
      False Float=0
     )
    ))
   &CurlNoiseForce(
    Bias Noise=true
    Bias Noise Field:Vector3f=<1 1 0>
    Mask Curl Noise=true
    Noise Frequency=1
    Noise Strength=&MakeCustomFloatFromBool(
     Bool=$Particles.Is Active
     True Float=&FloatFromCurve(FloatCurve=DataInterface() {propertyValues {Curve {keys [<at 0 1>,<at 0.1 -1>]}}}, Scale Curve=2500)
     False Float=0
    )
    Pan Noise=false
   ) {enabled=false}
   &SolveForcesAndVelocity(Clamp Velocity=true, Limit Acceleration=false, Speed Limit=250)
   &ScaleMeshSize(Scale Factor=&MakeVector(
     X=1
     Y=1
     Z=&Waveform(
      "Waveform [1]"=::NewEnumerator0
      "[1] Phase Scale"=&RandomRangeFloat(Minimum=0.4, Maximum=3)
      "[1] Frequency"=&RandomRangeFloat(Minimum=0.75, Maximum=1)
      "[1] Amplitude Min/Max"=<0.7 2>
     )
    )) {module=ScaleMeshSize001}
   &KillParticles(Kill Particles=&self:ScratchDynamicInput_01())
   &Collision(
    Age Colliding Particles=false
    Collision Enabled=&SetBoolByFloatComparison(A=$Particles.Age, B=0.25)
    Enable Rest State=false
    Friction=0
    GPU Collision Type=::NewEnumerator2
    Kill Particles Lodged Within Meshes=false
    KillOnCollision=false
    Max Number Of Collisions=2
    Module.EnableMaxCollisionCount=false
    Module.Randomize Collision Normal Vector=true
    Particle Radius Scale=5
    Randomize Collision Normal:NiagaraFloat=&RandomRangeFloat(Minimum=0.05, Maximum=0.15)
    Restitution=&RandomRangeFloat(Minimum=1, Maximum=1)
    Simple Friction=false
    Static Friction=0
    Static Friction Engagement Speed=0
   )
   &SolveForcesAndVelocity(Clamp Velocity=true, Limit Acceleration=false, Speed Limit=500) {module=SolveForcesAndVelocity001}
   &DynamicMaterialParameters(
    Index 0 Param 1=&MakeCustomFloatFromBool(
     Bool=$Particles.Is Active
     True Float=&RandomRangeFloat(Minimum=1.1, Maximum=2)
     False Float=&Multiply_Float(
      A=&Waveform(
       "Waveform [1]"=::NewEnumerator0
       "[1] Frequency"=&RandomRangeFloat(Minimum=0.25, Maximum=1)
       "[1] Amplitude Min/Max"=<0.5 2>
      )
      B=&RandomRangeFloat(Minimum=0.4, Maximum=0.8)
     )
    )
    Index 0 Param 2=&Multiply_Float(
     A=&MakeCustomFloatFromBool(
      Bool=$Particles.Is Active
      True Float=&RandomRangeFloat(Minimum=1, Maximum=2)
      False Float=&RandomRangeFloat(Minimum=0.5, Maximum=1)
     )
     B=&RandomRangeFloat(Minimum=2, Maximum=3)
    )
    Index 0 Param 3=&RandomRangeFloat(Minimum=1, Maximum=2)
    Index 1 Param 1=&SelectFloatFromArray(
     Array Sampling Mode=::NewEnumerator0
     Direct Array Index=?
     Float Selection Array=DataInterface() {propertyValues {FloatData [0,0.5,1]}}
     Random Seed=0
    )
    Index 1 Param 2=&RandomRangeFloat(Minimum=2, Maximum=4)
    Index 1 Param 3=/Niagara/DynamicInputs/Multiply/ScaleAndBiasFloat(Float:NiagaraFloat=1, Scale:NiagaraFloat=1, Bias:NiagaraFloat=0.1)
    Module.Param0WriteEnabled=true
    Module.Param0WriteEnabled1=true
    Module.Param1WriteEnabled=true
    Module.Param1WriteEnabled1=true
    Module.Param2WriteEnabled=true
    Module.Param2WriteEnabled1=true
    Module.Param3WriteEnabled=true
    Module.Param3WriteEnabled1=true
    Write Parameter Index 1=true
   )
   &ScaleColor(
    Scale Alpha=&Multiply_Float(
     A=&RandomRangeFloat(Minimum=25, Maximum=25)
     B=&FloatFromCurve(
      FloatCurve=DataInterface() {propertyValues {Curve {keys [<at 0 1>,<at 1 0>]}}}
      CurveIndex=&FN_LOD_Distance_Remap(Start Distance=1000, End Distance=1600, Minimum Output=0, Maximum Output=0, Clamp Output=true)
     )
    )
    Scale RGB=<3 3 3>
    ScaleA=true
    ScaleRGB=true
   )
   &ScaleMeshSize(Scale Factor=&MakeVector(
     X=1
     Y=1
     Z=&FloatFromCurve(
      FloatCurve=DataInterface() {propertyValues {Curve {keys [<at 0 1>,<at 0.22 2>,<at 0.25 1>]}}}
      Scale Curve=&Waveform(
       "Waveform [1]"=::NewEnumerator0
       "[1] Frequency"=&RandomRangeFloat(Minimum=3, Maximum=5)
       "[1] Amplitude Min/Max"=<0.6 0.9>
      )
     )
    ))
   &ScaleMeshSize(Scale Factor=&Multiply_VectorByFloat(
     Vector=<1 1 1>
     Float=&FloatFromCurve(FloatCurve=DataInterface() {propertyValues {Curve {keys [<at 0.75 1>,<at 1 0>]}}})
    )) {module=ScaleMeshSize002}
   &self:ScratchModule_02(ZDampening=0.5) {module=ScratchModule_02}
   &KillParticlesInVolume(Invert Volume=true, Sphere Radius=5000)
  ]
  rendererClasses [&NiagaraMeshRendererProperties]
  renderers [{rendererClass=&NiagaraMeshRendererProperties, rendererIndex=0}]
 }}
summary {
 systemName=NS_SpriteHaven_Frog_OnLandScape
 emitters [{emitterName=Frogs, simTarget=GPUComputeSim, rendererClasses [&NiagaraMeshRendererProperties]}]
 userVariables [
  User.BoxSize {type=Vector3f, description="", defaultValue {struct=Vector3f, value=<0 0 0>}}
  User.Radius Mult {type=NiagaraFloat, defaultValue=1, description=""}
  User.Spawn Mult {type=NiagaraFloat, defaultValue=1, description=""}
 ]
}
