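# Niagara environment effect for Hera's SpriteHaven that scatters GPU particles over landscape within a cylinder, filtered by height and normal.
# The active Onground emitter spawns short-lived cyan sprites that drift upward under curl noise, inverted gravity, wind and drag.
# An inverted view mask modulates size and alpha, while distance LOD shrinks derived sizes and colors are varied for three sprite renderers.
# The Mushrooms emitter, with landscape-aligned glowing meshes, is disabled.
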NiagaraSystem /Game/Environments/Hera/FX/SpriteHaven/NS_SpriteHaven_Mushrooms
dependencies {
 usedDataInterfaces [
  &NiagaraDataInterfaceLandscape
  &NiagaraDataInterfaceColorCurve
  &NiagaraDataInterfaceCurve
  &NiagaraDataInterfaceCollisionQuery
 ]
 usedDynamicInputs [&Vector2DFromFloat]
 usedModules [&SystemState]
}
emitters {
 Mushrooms {
  bEnabled=false
  simTarget=GPUComputeSim
  emitterUpdateScript [
   &EmitterState()
   &SpawnRate(SpawnRate=500, Use Spawn Probability=false)
  ]
  particleSpawnScript [
   &InitializeParticle(
    Color Maximum=<rgba 0 0.0873 1 1>
    Color Minimum=<rgba 1 0 0.8196 1>
    Color Mode=::NewEnumerator2
    Lifetime=5
    Mesh Scale Mode=::NewEnumerator1
    Mesh Uniform Scale Max=0.2
    Mesh Uniform Scale Min=0.1
    Sprite Size Mode=::NewEnumerator1
    Uniform Sprite Size Max=750
    Uniform Sprite Size Min=500
    UsePositionOffset=false
    Write Material Random=true
   )
   &ShapeLocation(Cylinder Radius=6000, MaskSpawnGroup=false, Override Seed=false, Shape Primitive=::NewEnumerator1, Surface Only=false, UseOwnerScale=true)
   /Niagara/Modules/Landscape/SampleLandscape(
    Facing Axis:Vector3f=<0 0 1>
    Height Filter=true
    Normal Filter=true
    Normal Filter Tolerance=0.2
    Orient Mesh To Normal=true
    Use Max Height=true
    Use Min Height=true
   )
   &InitialMeshOrientation(
    Mesh Orientation Mode=::NewEnumerator1
    Rotation=&RandomRangeVector(Minimum=<-0.1 -0.1 -0.25>, Maximum=<0.1 0.1 0.25>)
    Use Rotation Vector=true
   )
  ]
  particleUpdateScript [
   &ParticleState()
   &UpdateMeshOrientation(Rotation Vector=<0.02 0.02 0.2>)
   &ScaleMeshSize(Scale Factor=<2 2 2>)
   &ScaleColor(
    Scale Alpha=&FloatFromCurve(
     FloatCurve=DataInterface() {propertyValues {Curve {keys [
         {interpMode=RCIM_Cubic, tangentMode=RCTM_User, time=0, value=0}
         {interpMode=RCIM_Cubic, tangentMode=RCTM_User, time=0.5, value=1}
         {tangentMode=RCTM_User, time=1, value=0}
        ]}}}
     Scale Curve=&Multiply_Float(
      A=250
      B=/Niagara/DynamicInputs/Camera/CameraFade(
       Camera Query:NiagaraDataInterfaceCamera=?
       Fade Offset:NiagaraFloat=2500
       Fade Distance:NiagaraFloat=2500
       Use Owner Position:NiagaraBool=true
      )
     )
    )
    Scale RGB=<25000 0 25000>
    ScaleA=true
    ScaleRGB=true
   )
   &AccelerationForce(Acceleration=<150 0 25>) {enabled=false}
   &Drag(Use Linear Drag=true, Use Rotational Drag=true) {enabled=false}
   &SolveForcesAndVelocity(Clamp Velocity=false, Limit Acceleration=false)
   &CameraOffset(Camera Offset Amount=15)
  ]
  rendererClasses [&NiagaraMeshRendererProperties]
  renderers [
   {rendererClass=&NiagaraMeshRendererProperties, rendererIndex=0}
   {rendererClass=&NiagaraMeshRendererProperties, rendererIndex=1}
  ]
 }
 Onground {
  simTarget=GPUComputeSim
  emitterUpdateScript [
   &EmitterState()
   &SpawnRate(SpawnRate=500, Use Spawn Probability=false)
  ]
  particleSpawnScript [
   &InitializeParticle(
    Color Maximum=<rgba 0 1 0.9371 1>
    Color Minimum=<rgba 0 0.7842 1 1>
    Color Mode=::NewEnumerator2
    Lifetime=&UniformRangedFloat(Minimum=3, Maximum=5)
    Mass Mode=::NewEnumerator1
    Position=&SimulationPosition()
    Position Mode=::NewEnumerator1
    Sprite Size Mode=::NewEnumerator1
    Uniform Sprite Size Max=20
    UsePositionOffset=false
    Write Material Random=true
   ) {module=InitializeParticle001}
   &AddVelocity(
    Cone Angle=90
    Cone Axis=<0 0 1>
    Module.UseConeFalloff=false
    Module.UseVelDistribution=false
    Override Seed=false
    Velocity Mode=::NewEnumerator2
    Velocity Speed=&RandomRangeFloat(Minimum=1, Maximum=250)
   )
   &InitialMeshOrientation(Mesh Axis To Orient=<0 0 1>, Mesh Orientation Mode=::NewEnumerator2, Orientation Vector=<0 0 1>, Use Rotation Vector=false)
   &ShapeLocation(Cylinder Radius=6000, MaskSpawnGroup=false, Override Seed=false, Shape Primitive=::NewEnumerator1, Surface Only=false, UseOwnerScale=true)
   &SampleLandscape(
    Facing Axis:Vector3f=<0 0 1>
    Height Filter=true
    Height Max=2500
    Height Min=10
    Normal Filter=true
    Normal Filter Tolerance=0.2
    Orient Mesh To Normal=true
    Use Max Height=true
    Use Min Height=true
   )
  ]
  particleUpdateScript [
   &ParticleState()
   /Niagara/Modules/Update/Camera/ViewMask(Border Percent Max=50, Border Percent Min=35)
   &ScaleSpriteSize(
    Uniform Curve Scale=&Multiply_Float(A=0.5, B=&OneMinusFloat(Float=$Output.ViewMask.ViewMask))
    Uniform Curve Sprite Scale=DataInterface() {propertyValues {Curve {keys [{interpMode=RCIM_Cubic, tangentMode=RCTM_User, time=0, value=0}, <at 0.2005 0.9765 RCIM_Cubic>, {tangentMode=RCTM_User, time=1, value=0}]}}}
   )
   &ScaleColor(
    Scale Alpha=&FloatFromCurve(
     FloatCurve=DataInterface() {propertyValues {Curve {keys [
         {interpMode=RCIM_Cubic, tangentMode=RCTM_User, time=0, value=0}
         {interpMode=RCIM_Cubic, tangentMode=RCTM_User, time=0.5, value=1}
         {tangentMode=RCTM_User, time=1, value=0}
        ]}}}
     Scale Curve=&Multiply_Float(A=0.4, B=&OneMinusFloat(Float=$Output.ViewMask.ViewMask))
    )
    ScaleA=true
    ScaleRGB=false
   )
   /Niagara/Modules/Update/Forces/CurlNoiseForce(
    Noise Frequency:NiagaraFloat=&RandomRangeFloat(Minimum=0.002, Maximum=0.005)
    Noise Strength=&RandomRangeFloat(Minimum=250, Maximum=2500)
    Pan Noise=true
    Pan Noise Field=<0 1 1>
   )
   &GravityForce(Gravity=<0 0 800>)
   &WindForce(Wind Speed=<250 0 25>)
   &Drag(Drag=0.1, Use Linear Drag=true, Use Rotational Drag=true)
   &SolveForcesAndVelocity(Clamp Velocity=true, Limit Acceleration=false, Speed Limit=50)
   /SetVariables(Particles.NewSize:Vector2f=&Multiply_Vector2D_ByFloat(
     Float=&FN_LOD_Distance_Remap(Start Distance=10000, End Distance=20000, Minimum Output=1, Maximum Output=0, Clamp Output=true)
     Vector2D=&Multiply_Vector2D_ByFloat(Float=2, Vector2D=$Particles.SpriteSize)
    )) {isSetParameters=true}
   /SetVariables(Particles.Vector 2D:Vector2f=&Multiply_Vector2D(A=$Particles.NewSize, B=<1 1>)) {isSetParameters=true}
   /SetVariables(
    Particles.Linear Color=&MultiplyLinearColor(Linear Color=$Particles.Color, Scale Factor=<1.5 1.5 1 0.5>)
    Particles.Linear Color001:LinearColor=&MultiplyLinearColor(Linear Color=$Particles.Color, Scale Factor=<250 250 1 1>)
   ) {isSetParameters=true}
   &CameraOffset(Camera Offset Amount=1500)
   /SetVariables(Particles.Vector 2D001:Vector2f=&Multiply_Vector2D(A=$Particles.NewSize, B=<0.5 0.5>)) {isSetParameters=true}
  ]
  rendererClasses [&NiagaraSpriteRendererProperties]
  renderers [
   {rendererClass=&NiagaraSpriteRendererProperties}
   {rendererClass=&NiagaraSpriteRendererProperties, rendererIndex=1}
   {rendererClass=&NiagaraSpriteRendererProperties, rendererIndex=2}
  ]
 }
}
summary {
 systemName=NS_SpriteHaven_Mushrooms
 emitters [
  {bEnabled=false, emitterName=Mushrooms, simTarget=GPUComputeSim, rendererClasses [&NiagaraMeshRendererProperties]}
  {emitterName=Onground, simTarget=GPUComputeSim, rendererClasses [&NiagaraSpriteRendererProperties]}
 ]
 userVariables [User.Region_ParticleSize {type=NiagaraFloat, defaultValue=8000, description=""}]
}
