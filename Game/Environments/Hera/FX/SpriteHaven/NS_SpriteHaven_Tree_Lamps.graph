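# Niagara system that places glowing lamps on a SpriteHaven tree, tinted between two user base colors.
# A single emitter bursts persistent particles at every fourth vertex of a lamp-position mesh via a sequential index array, offset along normals.
# Update gently rocks meshes, pulses sprite size, boosts color and derives distance-based sizes and camera offsets for one mesh and three sprite renderers.
# Time-of-day spawn rate and curl noise are present but disabled.
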NiagaraSystem /Game/Environments/Hera/FX/SpriteHaven/NS_SpriteHaven_Tree_Lamps
dependencies {
 usedDataInterfaces [&NiagaraDataInterfaceCurve, &NiagaraDataInterfaceStaticMesh, &NiagaraDataInterfaceArrayInt32]
 usedDynamicInputs [&Vector2DFromFloat]
 usedModules [
  &SystemState
  &self:SystemSpawnScript.SystemScriptSource.SystemScriptGraph.NiagaraNodeAssignment_0.SetVariables_#1
 ]
}
emitters {Parent {
  emitterUpdateScript [
   &EmitterState(Life Cycle Mode=::NewEnumerator1, Loop Behavior=::NewEnumerator1, Loop Duration Mode=::NewEnumerator1, UseLoopDelay=false)
   &SpawnBurst_Instantaneous(Spawn Count=39, Use Spawn Probability=false, UseLoopCountLimit=false)
   &SpawnRate(
    SpawnRate=&Multiply_Float(
     A=500
     B=&FloatFromCurve(
      FloatCurve=DataInterface() {propertyValues {Curve {keys [<at 0 1>,<at 0.2 1>,<at 0.8 1>,<at 1 1>]}}}
      CurveIndex=Linked() {linkedVariable {name=NPC.FortniteNPC.TimeOfDay_0-1, type=NiagaraFloat}}
     )
    )
    Use Spawn Probability=false
   ) {enabled=false}
  ]
  particleSpawnScript [
   &InitializeParticle(
    Color Maximum=$User.BaseColor2
    Color Minimum=$User.BaseColor
    Color Mode=::NewEnumerator2
    Lifetime=&UniformRangedFloat(Minimum=1, Maximum=1)
    Mass Mode=::NewEnumerator1
    Mesh Scale Mode=::NewEnumerator3
    Mesh Uniform Scale=1.5
    Position=&SimulationPosition()
    Position Mode=::NewEnumerator1
    Sprite Size Mode=::NewEnumerator1
    Uniform Sprite Size Max=25
    Uniform Sprite Size Min=25
    UsePositionOffset=false
    Write Material Random=true
   ) {module=InitializeParticle001}
   &InitialMeshOrientation(
    Mesh Axis To Orient=<0 0 1>
    Mesh Orientation Mode=::NewEnumerator2
    Orientation Vector=<0 0 1>
    Rotation=&RandomRangeVector(Minimum=<0 0 0>, Maximum=<0 0 1>)
    Use Rotation Vector=true
   )
   /Niagara/Modules/Spawn/Location/StaticMeshLocation(
    MaskSpawnGroup=false
    Mesh Sampling Type:ENiagara_StaticSamplingMode=::NewEnumerator2
    Offset Position Along Sampled Normal=-100
    OffsetAlongNormal=true
    Orientation Sampling:ENiagara_AttributeSamplingApplyOutput=::NewEnumerator1
    Position Sampling:ENiagaraSkelMeshPositionSamplingMode=::NewEnumerator1
    SamplePosOffset=true
    Static Mesh=DataInterface() {propertyValues {
      DefaultMesh=null
      InstanceIndex=-1
      LODIndex=1
      PreviewMesh=/Game/Environments/Hera/FX/Meshes/SM_Tree_Lamps_MeshParticle_Positions_01
      SoftSourceActor=null
      SourceMode=Default
      bAllowSamplingFromStreamingLODs=false
      bCaptureTransformsPerFrame=true
      bUsePhysicsBodyVelocity=false
      FilteredSockets []
      LODIndexUserParameter {parameter {name=None, typeDefHandle {registeredTypeIndex=108}, varData []}}
      MeshParameterBinding {parameter {name=None, typeDefHandle {registeredTypeIndex=1}, varData []}}
      SectionFilter {allowedMaterialSlots []}
     }}
    Vertex ID:NiagaraInt32=&SelectIntFromArray(
     Direct Array Index=/CascadeToNiagaraConverter/NiagaraScripts/DynamicInputScripts/SequentialInt(
      Use Array Data:NiagaraBool=false
      First Number:NiagaraInt32=0
      Sequence:NiagaraDataInterfaceArrayInt32=DataInterface() {propertyValues {IntData []}}
      Last Number:NiagaraInt32=39
     )
     Int Selection Array=DataInterface() {propertyValues {IntData [0,3,7,11,15,19,23,27,31,35,39,43,47,51,55,59,63,67,71,75,79,83,87,91,95,99,103,107,111,115,119,123,127,131,135,139,143,147,151,155]}}
    )
    Vertex Sampling Mode:ENiagara_StaticVertexMode=::NewEnumerator1
   )
   &MeshRotationForce(Coordinate Space:ENiagaraMeshTransforms=::NewEnumerator1, Rotation=<0 0 0.2>)
   &ApplyInitialForces()
  ]
  particleUpdateScript [
   &ParticleState(Kill Particles When Lifetime Has Elapsed=false, Let Infinitely Lived Particles Die When Emitter Deactivates=true)
   &UpdateMeshOrientation(
    Rotation Coordinate Space=::NewEnumerator1
    Rotation Rate=&Cosine(Normalized Angle=$Emitter.Age, Period=&RandomRangeFloat(Minimum=4, Maximum=5), Scale=0.1, Bias=0)
    Rotation Vector=&RandomRangeVector(Minimum=<0 0 0>, Maximum=<1 1 1>)
   ) {module=UpdateMeshOrientation001}
   &ScaleSpriteSize(
    Uniform Curve Scale=&Multiply_Float(
     A=0.5
     B=&Waveform(
      "Waveform [1]"=::NewEnumerator0
      "[1] Frequency"=&RandomRangeFloat(Minimum=0.5, Maximum=0.8)
      "[1] Amplitude Min/Max"=<0.7 1>
     )
    )
    Uniform Curve Sprite Scale=DataInterface() {propertyValues {Curve {keys [{tangentMode=RCTM_User, time=0, value=0}, {tangentMode=RCTM_User, time=0.0156, value=0.9059}, <at 0.067 0.9324>, <at 0.9783 1>]}}}
   )
   &ScaleColor(
    Scale Alpha=&FloatFromCurve(
     FloatCurve=DataInterface() {propertyValues {Curve {keys [
         {interpMode=RCIM_Cubic, tangentMode=RCTM_User, time=0, value=0}
         {interpMode=RCIM_Cubic, tangentMode=RCTM_User, time=0.0228, value=1}
         {tangentMode=RCTM_User, time=1, value=1}
        ]}}}
     Scale Curve=0.1
    )
    Scale RGB=<250 250 250>
    ScaleA=true
    ScaleRGB=true
   )
   /Niagara/Modules/Update/Forces/CurlNoiseForce(
    Noise Frequency:NiagaraFloat=&RandomRangeFloat(Minimum=1, Maximum=1)
    Noise Strength=&RandomRangeFloat(Minimum=250, Maximum=500)
    Pan Noise=true
    Pan Noise Field=<0 1 1>
   ) {enabled=false}
   &SolveForcesAndVelocity(Clamp Velocity=true, Limit Acceleration=false, Speed Limit=35)
   &CalculateAccurateVelocity() {enabled=false}
   &UpdateMeshOrientation(Rotation Rate=&Waveform(
     Global Input Position=&RandomRangeFloat(Minimum=0, Maximum=1)
     Global Amplitude Scale=0.5
     Global Clamp Min=-1
     "Waveform [1]"=::NewEnumerator0
     "[1] Phase Scale"=0.2
     "[1] Frequency"=1
     "[1] Amplitude Min/Max"=<-0.2 0.2>
     "[1] Scale"=&RandomRangeFloat(Minimum=0.4, Maximum=0.5)
    )) {enabled=false}
   /SetVariables(Particles.NewSize:Vector2f=&Multiply_Vector2D_ByFloat(
     Float=&FN_LOD_Distance_Remap(Start Distance=10000, End Distance=25000, Minimum Output=1, Maximum Output=1, Clamp Output=true)
     Vector2D=&Multiply_Vector2D_ByFloat(Float=1, Vector2D=$Particles.SpriteSize)
    )) {isSetParameters=true}
   /SetVariables(
    Particles.Vector 2D:Vector2f=&Multiply_Vector2D(A=$Particles.NewSize, B=<8 8>)
    Particles.Vector 2D001:Vector2f=&Multiply_Vector2D(
     A=$Particles.NewSize
     B=&Multiply_Vector2D_ByFloat(
      Float=&FN_LOD_Distance_Remap(Start Distance=1500, End Distance=5000, Minimum Output=0, Maximum Output=2, Clamp Output=true)
      Vector2D=<3 3>
     )
    )
   ) {isSetParameters=true}
   /SetVariables(
    Particles.Linear Color=&MultiplyLinearColor(Linear Color=$Particles.Color, Scale Factor=<1 1 1 0.02>)
    Particles.Linear Color001:LinearColor=&MultiplyLinearColor(Linear Color=$Particles.Color, Scale Factor=<5 5 5 1>)
   ) {isSetParameters=true}
   &CameraOffset()
   &DynamicMaterialParameters(
    Index 0 Param 1=1
    Module.Index0Float0Write=true
    Module.Index0Float1Write=true
    Module.Index0Float2Write=true
    Module.Index0Float3Write=true
   )
   /SetVariables(Particles.float:NiagaraFloat=&Add_Float(
     A=$Particles.CameraOffset
     B=&Lerp_Float(
      A=-25
      B=80
      Alpha=&FN_LOD_Distance_Remap(Start Distance=1000, End Distance=2500, Minimum Output=0, Maximum Output=1, Clamp Output=true)
     )
    )) {isSetParameters=true}
   /SetVariables(Particles.Linear Color=&MultiplyLinearColor(Linear Color=$Particles.Color, Scale Factor=<1 1 1 0.25>)) {isSetParameters=true}
  ]
  rendererClasses [&NiagaraSpriteRendererProperties, &NiagaraMeshRendererProperties]
  renderers [
   {rendererClass=&NiagaraSpriteRendererProperties}
   {rendererClass=&NiagaraSpriteRendererProperties, rendererIndex=1}
   {rendererClass=&NiagaraMeshRendererProperties, rendererIndex=2}
   {rendererClass=&NiagaraSpriteRendererProperties, rendererIndex=3}
  ]
 }}
summary {
 systemName=NS_SpriteHaven_Tree_Lamps
 emitters [{emitterName=Parent, rendererClasses [&NiagaraSpriteRendererProperties, &NiagaraMeshRendererProperties]}]
 userVariables [
  User.BaseColor {type=LinearColor, description="", defaultValue {struct=LinearColor, value=<rgba 1 0.3202 0 1>}}
  User.BaseColor2 {type=LinearColor, description="", defaultValue {struct=LinearColor, value=<rgba 1 0.7324 0.0635 1>}}
  User.Landscape {
   type=NiagaraDataInterfaceLandscape
   description=""
   defaultValue {
    struct=NiagaraExt_VariableValue_DataInterface
    value {
     dataInterface=&self:NiagaraDataInterfaceLandscape_0
     dataInterfaceClass=&NiagaraDataInterfaceLandscape
    }
   }
  }
  User.Region_ParticleSize {type=NiagaraFloat, defaultValue=8000, description=""}
  User.SpawnRate {type=NiagaraFloat, defaultValue=1500, description=""}
  User.SphereLocation {type=NiagaraFloat, defaultValue=8000, description=""}
  User.UserColor {type=LinearColor, description="", defaultValue {struct=LinearColor, value=<rgba 0.552 0.1403 0.5056 1>}}
 ]
}
