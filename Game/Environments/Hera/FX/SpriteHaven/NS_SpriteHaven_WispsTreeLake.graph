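# Niagara GPU ambient effect for the SpriteHaven tree and lake, with drifting wisps and falling petals.
# Parent spawns landscape-sampled wisps in two shaped groups; OnTree fills a raised squashed sphere with cyan wisps.
# Region masks, pulsing waveforms, camera fade and distance remaps shape wisp size, while time-of-day curves, currently flat, feed spawn rate and alpha.
# Petals are terrain-offset meshes moved by gravity, wind, noise and aerodynamic drag, with distance culling.
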NiagaraSystem /Game/Environments/Hera/FX/SpriteHaven/NS_SpriteHaven_WispsTreeLake
dependencies {
 usedDataInterfaces [&NiagaraDataInterfaceCurve, &NiagaraDataInterfaceCollisionQuery]
 usedDynamicInputs [&Vector2DFromFloat]
 usedModules [
  &SystemState
  &self:SystemSpawnScript.SystemScriptSource.SystemScriptGraph.NiagaraNodeAssignment_0.SetVariables_#1
 ]
}
emitters {
 OnTree {
  simTarget=GPUComputeSim
  emitterUpdateScript [
   &EmitterState()
   &SpawnRate(
    SpawnRate=&Multiply_Float(
     A=25
     B=&FloatFromCurve(
      FloatCurve=DataInterface() {propertyValues {Curve {keys [<at 0 1>,<at 0.2 1>,<at 0.8 1>,<at 1 1>]}}}
      CurveIndex=Linked() {linkedVariable {name=NPC.FortniteNPC.TimeOfDay_0-1, type=NiagaraFloat}}
     )
    )
    Use Spawn Probability=false
   )
  ]
  particleSpawnScript [
   &InitializeParticle(
    Color Maximum=<rgba 0 1 0.9371 1>
    Color Minimum=<rgba 0 0.7842 1 1>
    Color Mode=::NewEnumerator2
    Lifetime=&UniformRangedFloat(Minimum=3, Maximum=5)
    Mass Mode=::NewEnumerator1
    Position=&SimulationPosition()
    Position Mode=::NewEnumerator1
    Sprite Size Mode=::NewEnumerator1
    Uniform Sprite Size Max=15
    UsePositionOffset=false
    Write Material Random=true
   ) {module=InitializeParticle001}
   &AddVelocity(
    Cone Angle=90
    Cone Axis=<0 0 1>
    Module.UseConeFalloff=false
    Module.UseVelDistribution=false
    Override Seed=false
    Velocity Mode=::NewEnumerator2
    Velocity Speed=&RandomRangeFloat(Minimum=1, Maximum=250)
   )
   &InitialMeshOrientation(Mesh Axis To Orient=<0 0 1>, Mesh Orientation Mode=::NewEnumerator2, Orientation Vector=<0 0 1>, Use Rotation Vector=false)
   &ShapeLocation(
    MaskSpawnGroup=false
    Non Uniform Scale=<1 1 0.5>
    Offset=<0 0 3000>
    Offset Mode=::NewEnumerator0
    Override Seed=false
    Spawn Group Mask=1
    Sphere Radius=2250
    UseOwnerScale=true
   ) {module=ShapeLocation001}
   &KillParticlesInVolume(Sphere Radius=1500) {enabled=false}
  ]
  particleUpdateScript [
   &ParticleState()
   &RegionMask(Inner Float=5, Region Radius:NiagaraFloat=$User.Region_ParticleSize, UseOffset=true, UseRotation=false)
   &RegionMask(Region Radius:NiagaraFloat=$User.Region_ParticleSize, UseOffset=true, UseRotation=false) {module=RegionMask001}
   &ScaleSpriteSize(
    Uniform Curve Scale=&Multiply_Float(
     A=$Particles.RegionMask.Float
     B=&Multiply_Float(
      A=&Lerp_Float(
       A=0.25
       B=1
       Alpha=/Niagara/DynamicInputs/Camera/CameraFade(
        Camera Query:NiagaraDataInterfaceCamera=?
        Fade Offset:NiagaraFloat=500
        Fade Distance:NiagaraFloat=2500
        Use Owner Position:NiagaraBool=true
       )
      )
      B=&Multiply_Float(
       A=0.5
       B=&Waveform(
        "Waveform [1]"=::NewEnumerator0
        "[1] Frequency"=&RandomRangeFloat(Minimum=0.5, Maximum=0.7)
        "[1] Amplitude Min/Max"=<0.5 1>
       )
      )
     )
    )
    Uniform Curve Sprite Scale=DataInterface() {propertyValues {Curve {keys [
        {interpMode=RCIM_Cubic, tangentMode=RCTM_User, time=0, value=0}
        {arriveTangent=1.1111, interpMode=RCIM_Cubic, leaveTangent=1.1111, time=0.2, value=1}
        {interpMode=RCIM_Cubic, tangentMode=RCTM_User, time=0.9, value=1}
        {tangentMode=RCTM_User, time=1, value=0}
       ]}}}
   )
   &ScaleColor(
    Scale Alpha=&Multiply_Float(
     A=&FloatFromCurve(
      FloatCurve=DataInterface() {propertyValues {Curve {keys [
          {interpMode=RCIM_Cubic, tangentMode=RCTM_User, time=0, value=0}
          {interpMode=RCIM_Cubic, tangentMode=RCTM_User, time=0.5, value=1}
          {tangentMode=RCTM_User, time=1, value=0}
         ]}}}
      Scale Curve=&Multiply_Float(
       A=0.5
       B=&FloatFromCurve(
        FloatCurve=DataInterface() {propertyValues {Curve {keys [<at 0 1>,<at 0.2 1>,<at 0.8 1>,<at 1 1>]}}}
        CurveIndex=Linked() {linkedVariable {name=NPC.FortniteNPC.TimeOfDay_0-1, type=NiagaraFloat}}
        Scale Curve=0.5
       )
      )
     )
     B=$Particles.RegionMask001.Float
    )
    Scale RGB=&MakeVectorFromLinearColorRGB(Color=&HueShiftLinearColor(
      Hue=&Multiply_Float(A=$Particles.RegionMask001.Float, B=0.025)
      Saturation=1
      Value=0.5
      Linear Color=$Particles.Initial.Color
     ))
    ScaleA=true
    ScaleRGB=true
   )
   /Niagara/Modules/Update/Forces/CurlNoiseForce(
    Noise Frequency:NiagaraFloat=&RandomRangeFloat(Minimum=0.002, Maximum=0.005)
    Noise Strength=&RandomRangeFloat(Minimum=250, Maximum=2500)
    Pan Noise=true
    Pan Noise Field=<0 1 1>
   )
   &GravityForce()
   &WindForce(Wind Speed=<2500 0 0>)
   &Drag(Drag=0.1, Use Linear Drag=true, Use Rotational Drag=true)
   &SolveForcesAndVelocity(Clamp Velocity=true, Limit Acceleration=false, Speed Limit=50)
   &CalculateAccurateVelocity() {enabled=false}
   /SetVariables(Particles.NewSize:Vector2f=&Multiply_Vector2D_ByFloat(
     Float=&FN_LOD_Distance_Remap(Start Distance=10000, End Distance=20000, Minimum Output=1, Maximum Output=0, Clamp Output=true)
     Vector2D=&Multiply_Vector2D_ByFloat(Float=3, Vector2D=$Particles.SpriteSize)
    )) {isSetParameters=true}
   /SetVariables(Particles.Vector 2D:Vector2f=&Multiply_Vector2D(A=$Particles.NewSize, B=<1 1>)) {isSetParameters=true}
   /SetVariables(Particles.Linear Color=&MultiplyLinearColor(Linear Color=$StackContext.Color, Scale Factor=<1 1 1 1>)) {enabled=false, isSetParameters=true, module=SetVariables_#2}
   &CameraOffset(Camera Offset Amount=&RandomRangeFloat(Minimum=400, Maximum=600))
  ]
  rendererClasses [&NiagaraSpriteRendererProperties]
  renderers [
   {rendererClass=&NiagaraSpriteRendererProperties}
   {rendererClass=&NiagaraSpriteRendererProperties, rendererIndex=1}
  ]
 }
 Parent {
  simTarget=GPUComputeSim
  emitterUpdateScript [
   &EmitterState()
   &SpawnRate(
    SpawnRate=&Multiply_Float(
     A=200
     B=&FloatFromCurve(
      FloatCurve=DataInterface() {propertyValues {Curve {keys [<at 0 1>,<at 0.2 1>,<at 0.8 1>,<at 1 1>]}}}
      CurveIndex=Linked() {linkedVariable {name=NPC.FortniteNPC.TimeOfDay_0-1, type=NiagaraFloat}}
     )
    )
    Use Spawn Probability=false
   )
  ]
  particleSpawnScript [
   &InitializeParticle(
    Color Maximum=<rgba 0.0155 0.4479 0.2365 1>
    Color Minimum=<rgba 0.07 0.5521 0.4926 1>
    Color Mode=::NewEnumerator2
    Lifetime=&UniformRangedFloat(Minimum=3, Maximum=5)
    Mass Mode=::NewEnumerator1
    Position=&SimulationPosition()
    Position Mode=::NewEnumerator1
    Sprite Size Mode=::NewEnumerator1
    Uniform Sprite Size Max=15
    UsePositionOffset=false
    Write Material Random=true
   ) {module=InitializeParticle001}
   &FN_SetSpawnGroup(Number of Spawn Groups=2)
   &AddVelocity(
    Cone Angle=90
    Cone Axis=<0 0 1>
    Module.UseConeFalloff=false
    Module.UseVelDistribution=false
    Override Seed=false
    Velocity Mode=::NewEnumerator2
    Velocity Speed=&RandomRangeFloat(Minimum=1, Maximum=250)
   )
   /Niagara/Modules/Landscape/SampleLandscape(Height Offset=1500)
   &ShapeLocation(
    Handle Radius=5000
    Large Radius=5000
    MaskSpawnGroup=true
    Offset Mode=::NewEnumerator0
    Override Seed=false
    Shape Primitive=::NewEnumerator3
    UseOwnerScale=true
   )
   &InitialMeshOrientation(Mesh Axis To Orient=<0 0 1>, Mesh Orientation Mode=::NewEnumerator2, Orientation Vector=<0 0 1>, Use Rotation Vector=false)
   &ShapeLocation(MaskSpawnGroup=true, Offset Mode=::NewEnumerator0, Override Seed=false, Spawn Group Mask=1, Sphere Radius=2500, UseOwnerScale=true) {module=ShapeLocation001}
   &KillParticlesInVolume(Sphere Radius=1500)
  ]
  particleUpdateScript [
   &ParticleState()
   &RegionMask(Inner Float=5, Region Radius:NiagaraFloat=$User.Region_ParticleSize, UseOffset=true, UseRotation=false)
   &RegionMask(Region Radius:NiagaraFloat=$User.Region_ParticleSize, UseOffset=true, UseRotation=false) {module=RegionMask001}
   &ScaleSpriteSize(
    Uniform Curve Scale=&Multiply_Float(
     A=$Particles.RegionMask.Float
     B=&Multiply_Float(
      A=&Lerp_Float(
       A=0.25
       B=0.35
       Alpha=&CameraFade(
        Camera Query:NiagaraDataInterfaceCamera=?
        Fade Offset:NiagaraFloat=500
        Fade Distance:NiagaraFloat=2500
        Use Owner Position:NiagaraBool=true
       )
      )
      B=&Multiply_Float(
       A=0.5
       B=&Waveform(
        "Waveform [1]"=::NewEnumerator0
        "[1] Frequency"=&RandomRangeFloat(Minimum=0.5, Maximum=0.7)
        "[1] Amplitude Min/Max"=<0.5 1>
       )
      )
     )
    )
    Uniform Curve Sprite Scale=DataInterface() {propertyValues {Curve {keys [
        {interpMode=RCIM_Cubic, tangentMode=RCTM_User, time=0, value=0}
        {arriveTangent=1.1111, interpMode=RCIM_Cubic, leaveTangent=1.1111, time=0.2, value=1}
        {interpMode=RCIM_Cubic, tangentMode=RCTM_User, time=0.9, value=1}
        {tangentMode=RCTM_User, time=1, value=0}
       ]}}}
   )
   &ScaleColor(
    Scale Alpha=&Multiply_Float(
     A=&FloatFromCurve(
      FloatCurve=DataInterface() {propertyValues {Curve {keys [
          {interpMode=RCIM_Cubic, tangentMode=RCTM_User, time=0, value=0}
          {interpMode=RCIM_Cubic, tangentMode=RCTM_User, time=0.5, value=1}
          {tangentMode=RCTM_User, time=1, value=0}
         ]}}}
      Scale Curve=&Multiply_Float(
       A=1
       B=&FloatFromCurve(
        FloatCurve=DataInterface() {propertyValues {Curve {keys [<at 0 1>,<at 0.2 1>,<at 0.8 1>,<at 1 1>]}}}
        CurveIndex=Linked() {linkedVariable {name=NPC.FortniteNPC.TimeOfDay_0-1, type=NiagaraFloat}}
       )
      )
     )
     B=1
    )
    Scale RGB=&MakeVectorFromLinearColorRGB(Color=&HueShiftLinearColor(
      Hue=&Multiply_Float(A=$Particles.RegionMask001.Float, B=0.025)
      Saturation=1
      Value=0.1
      Linear Color=$Particles.Initial.Color
     ))
    ScaleA=true
    ScaleRGB=true
   )
   &CurlNoiseForce(
    Noise Frequency:NiagaraFloat=&RandomRangeFloat(Minimum=0.002, Maximum=0.005)
    Noise Strength=&RandomRangeFloat(Minimum=250, Maximum=500)
    Pan Noise=true
    Pan Noise Field=<0 1 1>
   )
   &SolveForcesAndVelocity(Clamp Velocity=true, Limit Acceleration=false, Speed Limit=35)
   &CalculateAccurateVelocity()
   /SetVariables(Particles.NewSize:Vector2f=&Multiply_Vector2D_ByFloat(
     Float=&FN_LOD_Distance_Remap(Start Distance=10000, End Distance=25000, Minimum Output=1, Maximum Output=0, Clamp Output=true)
     Vector2D=&Multiply_Vector2D_ByFloat(Float=5, Vector2D=$Particles.SpriteSize)
    )) {isSetParameters=true}
   /SetVariables(Particles.Vector 2D:Vector2f=&Multiply_Vector2D(A=$Particles.NewSize, B=<1 1>)) {isSetParameters=true}
   /SetVariables(Particles.Linear Color=&MultiplyLinearColor(Linear Color=$Particles.Color, Scale Factor=<1 1 1 0.25>)) {isSetParameters=true, module=SetVariables_#2}
  ]
  rendererClasses [&NiagaraSpriteRendererProperties]
  renderers [
   {rendererClass=&NiagaraSpriteRendererProperties}
   {rendererClass=&NiagaraSpriteRendererProperties, rendererIndex=1}
  ]
 }
 Petals {
  simTarget=GPUComputeSim
  emitterUpdateScript [
   &EmitterState(
    Enable Distance Culling=true
    Max Distance Response=::SleepAndClearParticles
    MaxDistance=24000
    Module.bUseMaxDistance=true
    Module.bUseMinDistance=false
    Reset Age On Awaken=true
    Scalability Mode=::NewEnumerator1
   )
   &SpawnRate(
    SpawnRate=&FloatFromCurve(
     FloatCurve=DataInterface() {propertyValues {Curve {keys [<at 0 1>,<at 0.1 2>,<at 0.25 1>,<at 0.4976 0.25>,<at 0.8283 0.25>,<at 1 1>]}}}
     CurveIndex=$System.NormalizedLoopAge
     Scale Curve=$User.SpawnRate
    )
    Use Spawn Probability=false
   )
  ]
  particleSpawnScript [
   &InitializeParticle(
    Color=&HueShiftLinearColor(
     Hue=&RandomRangeFloat(Minimum=-0.025, Maximum=0.025)
     Saturation=1
     Value=3
     Linear Color=$User.UserColor
    )
    Lifetime=&RandomRangeFloat(Minimum=4, Maximum=5)
    Mass=3
    Mass Mode=::NewEnumerator1
    Mesh Scale Mode=::NewEnumerator1
    Mesh Uniform Scale Max=150
    Mesh Uniform Scale Min=75
    Sprite Rotation Mode=::NewEnumerator2
    Sprite Size Mode=::NewEnumerator1
    Uniform Sprite Size Max=50
    Uniform Sprite Size Min=15
    UsePositionOffset=false
    Write Material Random=true
   )
   /SetVariables(Particles.Terrain Offset:NiagaraFloat=&RandomRangeFloat(Minimum=1600, Maximum=3000)) {isSetParameters=true}
   &ShapeLocation(
    MaskSpawnGroup=false
    Override Seed=false
    Sphere Radius=&Multiply_Float(A=$User.SphereLocation, B=1)
    UseOwnerScale=true
   )
   &SampleLandscape(Height Offset=500)
   /SetVariables(Particles.Position=&Add_Vector(
     A=Linked() {linkedVariable {name=Particles.Position, type=NiagaraPosition}}
     B=&MakeVector(X=0, Y=0, Z=$Particles.Terrain Offset)
    )) {isSetParameters=true}
   &CameraOffset(Camera Offset Amount=15)
   &SubUVAnimation(SubUV Animation Mode:ENiagara_SubUVLookupMode=::NewEnumerator2) {module=SubUVAnimation001}
   &InitialMeshOrientation(Rotation Coordinate Space=::NewEnumerator2, Use Rotation Vector=true)
   &ApplyInitialForces()
  ]
  particleUpdateScript [
   &ScaleColor(
    Scale Alpha=&FloatFromCurve(FloatCurve=DataInterface() {propertyValues {Curve {keys [<at 0.8 1>,<at 1 0>]}}}, Scale Curve=5)
    Scale RGB=&MakeVectorFromLinearColorRGB(Color=&MultiplyLinearColor(Linear Color=$Particles.Initial.Color, Scale Factor=&MakeVector4_FromFloat(Float=1)))
    ScaleA=true
    ScaleRGB=true
   )
   /SetVariables(Particles.Normalized Velocity:Vector3f=&NormalizeVector(Vector To Normalize=$Particles.Velocity)) {isSetParameters=true}
   &GravityForce(Gravity=&RandomRangeVector(Minimum=<0 0 -4500>, Maximum=<0 0 -6500>))
   &SolveForcesAndVelocity(Clamp Velocity=false, Limit Acceleration=false, Speed Limit=250)
   &SubUVAnimation(End Frame=64, SubUV Animation Mode:ENiagara_SubUVLookupMode=::NewEnumerator1, SubUV Loop Count=4)
   &ParticleState(DeltaTime:NiagaraFloat=&MakeCustomFloatFromBool(
     Bool=&SetBoolByFloatComparison(A=&VectorLength(Vector=$Particles.Velocity), B=1)
     True Float=$Engine.DeltaTime
     False Float=0
    ))
   &ScaleMeshSize(Scale Factor=&VectorFromFloat(Value=&Multiply_Float(
      A=&FloatFromCurve(
       FloatCurve=DataInterface() {propertyValues {Curve {keys [<at 0 0.25>,<at 500 1>,<at 6000 1>,<at 8000 0>]}}}
       CurveIndex=$Engine.Owner.LODDistance
       Scale Curve=0.5
      )
      B=&FloatFromCurve(
       FloatCurve=DataInterface() {propertyValues {Curve {keys [
           {interpMode=RCIM_Cubic, tangentMode=RCTM_User, time=0, value=0}
           {interpMode=RCIM_Cubic, tangentMode=RCTM_User, time=0.2447, value=0.85}
           <at 1 0>
          ]}}}
       Scale Curve=0.25
      )
     )))
   /Niagara/Modules/Update/Forces/V2/CurlNoiseForce(Bias Noise=true, Bias Noise Field:Vector3f=<1 0 0>, Noise Frequency=1, Noise Strength=25000, Pan Noise=false)
   &ScaleSpriteSize(
    Scale Sprite Size Mode=::NewEnumerator0
    Uniform Scale Factor=&Multiply_Float(
     A=&FloatFromCurve(
      FloatCurve=DataInterface() {propertyValues {Curve {keys [<at 0 1>,<at 500 1>,<at 6000 1>,<at 8000 0>]}}}
      CurveIndex=$Engine.Owner.LODDistance
      Scale Curve=2
     )
     B=&FloatFromCurve(FloatCurve=DataInterface() {propertyValues {Curve {keys [<at 0.8 1>,<at 1 0>]}}}, Scale Curve=0.25)
    )
   )
   &WindForce(
    Friction Mode:ENiagaraWindFrictionMode=::NewEnumerator0
    Offset By Time:Vector3f=<0 0 1>
    Scale=2
    Surface Alignment Mode:ENiagaraWindCombingMode=::NewEnumerator0
    Wind Speed=<250 0 -250>
    Wind Speed Scale=&RandomRangeFloat(Minimum=0.5, Maximum=1)
   ) {module=WindForce001}
   &AerodynamicDrag(
    Aerodynamic Drag=2
    Aerodynamic Rotational Drag=0.02
    Drag Shape Curve=DataInterface() {propertyValues {Curve {keys [
        {interpMode=RCIM_Cubic, tangentMode=RCTM_User, time=0, value=0.1}
        {arriveTangent=2.0799, interpMode=RCIM_Cubic, leaveTangent=2.0799, tangentMode=RCTM_User, time=1, value=1}
       ]}}}
    Lift Contribution=0
    Lift Shape Curve=DataInterface() {propertyValues {Curve {keys [
        {interpMode=RCIM_Cubic, tangentMode=RCTM_User, time=0, value=0}
        {interpMode=RCIM_Cubic, tangentMode=RCTM_User, time=0.5, value=1}
        {tangentMode=RCTM_User, time=1, value=0}
       ]}}}
    Pivot Offset Maximum:Vector3f=<5 1 1>
    Pivot Offset Minimum:Vector3f=<-1 -0.5 -2>
    Pivot Offset Minimum Length:Vector3f=<0 0 0>
    Rotation Strength:NiagaraFloat=&RandomRangeFloat(Minimum=0.2, Maximum=0.3)
    Use Linear Drag=true
    Use Rotational Drag=true
   )
   &SolveForcesAndVelocity(
    Clamp Velocity=true
    Limit Acceleration=false
    Speed Limit:NiagaraFloat=&RandomRangeFloat(Minimum=150, Maximum=250)
   ) {module=SolveForcesAndVelocity001}
   &AlignSpriteToMeshOrientation(
    Mesh Orientation Relative Sprite Alignment Vector=<1 0 0>
    Mesh Orientation Relative Sprite Facing Vector:Vector3f=<1 1 1>
    SetAlignment=true
    SetFacing=true
   )
   &DynamicMaterialParameters(
    Index 0 Param 1=&RandomRangeFloat(Minimum=0.02, Maximum=0.1)
    Index 0 Param 2=&RandomRangeFloat(Minimum=-0.05, Maximum=0.05)
    Index 0 Param 3=1
    Module.Param0WriteEnabled=true
    Module.Param1WriteEnabled=true
    Module.Param2WriteEnabled=true
    Module.Param3WriteEnabled=true
   )
  ]
  rendererClasses [&NiagaraMeshRendererProperties]
  renderers [{rendererClass=&NiagaraMeshRendererProperties, rendererIndex=0}]
 }
}
summary {
 systemName=NS_SpriteHaven_WispsTreeLake
 emitters [
  {emitterName=Parent, simTarget=GPUComputeSim, rendererClasses [&NiagaraSpriteRendererProperties]}
  {emitterName=OnTree, simTarget=GPUComputeSim, rendererClasses [&NiagaraSpriteRendererProperties]}
  {emitterName=Petals, simTarget=GPUComputeSim, rendererClasses [&NiagaraMeshRendererProperties]}
 ]
 userVariables [
  User.Landscape {
   type=NiagaraDataInterfaceLandscape
   description=""
   defaultValue {
    struct=NiagaraExt_VariableValue_DataInterface
    value {
     dataInterface=&self:NiagaraDataInterfaceLandscape_0
     dataInterfaceClass=&NiagaraDataInterfaceLandscape
    }
   }
  }
  User.Region_ParticleSize {type=NiagaraFloat, defaultValue=8000, description=""}
  User.SpawnRate {type=NiagaraFloat, defaultValue=5, description=""}
  User.SphereLocation {type=NiagaraFloat, defaultValue=2250, description=""}
  User.UserColor {type=LinearColor, description="", defaultValue {struct=LinearColor, value=<rgba 0.224 0.0361 0.0062 1>}}
 ]
}
